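# Control Rig for the hair of the female medium Pencil Fig head, adjusting hair target bones from head pose.
# A spherical pose reader on head, relative to spine_05, weights an offset applied to the left and right dyn_hair TGT bones.
# A separate chain aims each Look_at bone at its target, using Hair Aim controls as secondary targets, but shows no link from Begin Execution.
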ControlRigBlueprint /BRCosmetics/Characters/Player/Female/Medium/Heads/F_MED_Pencil_Fig_Head/Meshes/Hair/F_MED_PencilFig_Hair_CtrlRig
graphs [
 RigVMFunctionLibrary {links [], nodes []}
 RigVMModel {nodes [
   &RigUnit_ModifyTransforms(
    ExecutePin=()
    ItemToModify=((Item=(Type=Control,Name="Hair r Aim"),Transform=(Rotation=(X=0,Y=0,Z=0,W=1),Translation=(X=0,Y=0,Z=0),Scale3D=(X=1,Y=1,Z=1))),(Item=(Type=Control,Name="Hair l Aim"),Transform=(Rotation=(X=0,Y=0,Z=0,W=1),Translation=(X=0,Y=0,Z=0),Scale3D=(X=1,Y=1,Z=1))))
    Weight=1
   )
   > /Script/ControlRig.RigUnit_AimItem(
    DebugSettings=(Scale=10,WorldOffset=(Rotation=(X=0,Y=0,Z=0,W=1),Translation=(X=0,Y=0,Z=0),Scale3D=(X=1,Y=1,Z=1)))
    Item=(Type=Bone,Name="dyn_hair_l_main_Look_at")
    Primary=(Weight=1,Axis=(X=1,Y=0,Z=0),Target=(X=0,Y=1,Z=0),Kind=Location,Space=(Type=Bone,Name="dyn_hair_l_TGT"))
    Secondary=(Weight=1,Axis=(X=0,Y=-1,Z=0),Target=(X=1,Y=0,Z=0),Kind=Location,Space=(Type=Control,Name="Hair l Aim"))
    Weight=1
   ) {is=struct RigVMUnitNode}
   > &RigUnit_AimItem(
    DebugSettings=(Scale=10,WorldOffset=(Rotation=(X=0,Y=0,Z=0,W=1),Translation=(X=0,Y=0,Z=0),Scale3D=(X=1,Y=1,Z=1)))
    Item=(Type=Bone,Name="dyn_hair_r_main_Look_at")
    Primary=(Weight=1,Axis=(X=1,Y=0,Z=0),Target=(X=0,Y=1,Z=0),Kind=Location,Space=(Type=Bone,Name="dyn_hair_r_TGT"))
    Secondary=(Weight=1,Axis=(X=0,Y=-1,Z=0),Target=(X=1,Y=0,Z=0),Kind=Location,Space=(Type=Control,Name="Hair r Aim"))
    Weight=1
   )
   RigUnit_BeginExecution = &RigUnit_BeginExecution()
   n1 = &RigUnit_SphericalPoseReader(
    ActiveRegionScaleFactors=(PositiveWidth=1,NegativeWidth=0.5,PositiveHeight=1.5,NegativeHeight=1.5)
    ActiveRegionSize=0.1
    Debug=(bDrawDebug=false,DebugScale=25,DebugSegments=20,DebugThickness=0.25)
    DriverAxis=(X=1,Y=0,Z=0)
    DriverItem=(Type=Bone,Name="head")
    FalloffRegionScaleFactors=(PositiveWidth=1,NegativeWidth=1,PositiveHeight=1.5,NegativeHeight=1.5)
    FalloffSize=0.2
    FlipWidthScaling=true
    OptionalParentItem=(Type=Bone,Name="spine_05")
    RotationOffset=(X=-35,Y=-25,Z=35)
    ExecutePin=@RigUnit_BeginExecution.ExecutePin
   )
   > &RigUnit_ModifyTransforms(
    ItemToModify="((Item=(Type=Bone,Name=\"dyn_hair_r_TGT\"),Transform=(Rotation=(X=0,Y=0,Z=0,W=1),Translation=(X=-15,Y=-2,Z=0),Scale3D=(X=1,Y=1,Z=1))),(Item=(Type=Bone,Name=\"dyn_hair_l_TGT\"),Transform=(Rotation=(X=0,Y=0,Z=0,W=1),Translation=(X=-15,Y=-2,Z=0),Scale3D=(X=1,Y=1, ...(truncated)"
    Weight=@n1.OutputParam
   )
   &RigUnit_GetTransform(Item=(Type=Control,Name="Hair l Aim"))
  ]}
]
hierarchy [
 root {loc=<0 0 0>, rot=<0 0 0 1>}
 pelvis {parent=root, loc=<0 -0.5301 98.0571>, rot=<0.0127 0.707 -0.0127 -0.707>}
 spine_01 {parent=pelvis, loc=<6.1208 -0.9328 0>, rot=<0 0 0.034 -0.9994>}
 spine_02 {parent=spine_01, loc=<6.5996 -0.2095 0>, rot=<0 0 0.0238 -0.9997>}
 spine_03 {parent=spine_02, loc=<6.4914 -0.1237 0>, rot=<0 0 0.0149 -0.9999>}
 spine_04 {parent=spine_03, loc=<6.9598 0.0794 0>, rot=<0 0 -0.0132 -0.9999>}
 spine_05 {parent=spine_04, loc=<6.9987 0 0>, rot=<0 0 -0.0015 -1>}
 neck_01 {parent=spine_05, loc=<11.3138 0.1529 0>, rot=<0 0 -0.2003 -0.9797>}
 neck_02 {parent=neck_01, loc=<4.5627 0 0>, rot=<0 0 0.0504 -0.9987>}
 head {parent=neck_02, loc=<5.3914 0 0>, rot=<0 0 0.115 -0.9934>}
 dyn_hair_r_main_MSTR {parent=head, loc=<10.8876 -6.2822 6.8275>, rot=<-0.2933 0.6507 0.2825 -0.6409>}
 dyn_hair_r_main_Look_at {parent=dyn_hair_r_main_MSTR, loc=<0 0 0>, rot=<-0.31 0.6059 -0.2621 -0.6842>}
 dyn_hair_r_main_01 {parent=dyn_hair_r_main_Look_at, loc=<0 0 0>, rot=<0.31 -0.6059 0.2621 -0.6842>}
 dyn_hair_r_main_02 {parent=dyn_hair_r_main_01, loc=<2.3383 0 0>, rot=<-0.3535 0.5038 -0.1773 -0.768>}
 dyn_hair_r_main_03 {parent=dyn_hair_r_main_02, loc=<8.6686 0 0>, rot=<-0.0254 0.1014 -0.1385 -0.9848>}
 dyn_hair_r_main_04 {parent=dyn_hair_r_main_03, loc=<11.8196 0 0>, rot=<-0.0009 0.0115 -0.0268 -0.9996>}
 dyn_hair_r_main_05 {parent=dyn_hair_r_main_04, loc=<11.3236 0 0>, rot=<-0.0005 0.0386 -0.0385 -0.9985>}
 dyn_hair_r_main_06 {parent=dyn_hair_r_main_05, loc=<8.375 0 0>, rot=<-0.0002 -0.0221 0.0138 -0.9997>}
 dyn_hair_r_main_07 {parent=dyn_hair_r_main_06, loc=<9.0488 0 0>, rot=<-0.0009 0.693 -0.0009 -0.7209>}
 dyn_hair_r_a_01 {parent=dyn_hair_r_main_Look_at, loc=<-1.0651 -3.5458 -2.1317>, rot=<-0.2515 -0.0375 0.1382 -0.9572>}
 dyn_hair_r_a_02 {parent=dyn_hair_r_a_01, loc=<9.9656 0 0>, rot=<0.0869 0.0376 -0.1484 -0.9844>}
 dyn_hair_r_a_03 {parent=dyn_hair_r_a_02, loc=<12.0945 0 0>, rot=<-0.0077 0.0073 -0.0426 -0.999>}
 dyn_hair_r_a_04 {parent=dyn_hair_r_a_03, loc=<11.324 0 0>, rot=<-0.001 0.0044 -0.0164 -0.9999>}
 dyn_hair_r_a_05 {parent=dyn_hair_r_a_04, loc=<7.9003 0 0>, rot=<0.015 -0.0046 -0.0029 -0.9999>}
 dyn_hair_r_a_06 {parent=dyn_hair_r_a_05, loc=<7.6555 0 0>, rot=<0.1105 0.6817 -0.1075 -0.7153>}
 dyn_hair_l_main_MSTR {parent=head, loc=<10.8876 -6.2822 -6.8275>, rot=<0.2933 -0.6507 0.2825 -0.6409>}
 dyn_hair_l_main_Look_at {parent=dyn_hair_l_main_MSTR, loc=<0 0 0>, rot=<0.31 -0.6059 -0.2621 -0.6842>}
 dyn_hair_l_main_01 {parent=dyn_hair_l_main_Look_at, loc=<0 0 0>, rot=<-0.31 0.6059 0.2621 -0.6842>}
 dyn_hair_l_main_02 {parent=dyn_hair_l_main_01, loc=<2.3383 0 0>, rot=<0.3535 -0.5038 -0.1773 -0.768>}
 dyn_hair_l_main_03 {parent=dyn_hair_l_main_02, loc=<8.6686 0 0>, rot=<0.0254 -0.1014 -0.1385 -0.9848>}
 dyn_hair_l_main_04 {parent=dyn_hair_l_main_03, loc=<11.8196 0 0>, rot=<0.0009 -0.0115 -0.0268 -0.9996>}
 dyn_hair_l_main_05 {parent=dyn_hair_l_main_04, loc=<11.3236 0 0>, rot=<0.0005 -0.0386 -0.0385 -0.9985>}
 dyn_hair_l_main_06 {parent=dyn_hair_l_main_05, loc=<8.375 0 0>, rot=<0.0002 0.0221 0.0138 -0.9997>}
 dyn_hair_l_main_07 {parent=dyn_hair_l_main_06, loc=<9.0488 0 0>, rot=<-0.0009 0.7209 -0.0009 -0.693>}
 clavicle_l {parent=spine_05, loc=<5.4831 -0.1053 -3.2946>, rot=<0.0465 -0.7289 -0.009 -0.6829>}
 upperarm_l {parent=clavicle_l, loc=<9.8658 -1.2136 -0.4105>, rot=<0.0164 -0.3245 -0.0269 -0.9454>}
 upperarm_twist_01_l {parent=upperarm_l, loc=<7.2135 -0.1134 0>, rot=<0 0 0 1>}
 deform_upperarm_bk_base_l {parent=upperarm_twist_01_l, loc=<-6.4053 -4.3746 0.314>, rot=<0 0 0 -1>}
 deform_upperarm_bk_l {parent=deform_upperarm_bk_base_l, loc=<0 0 0>, rot=<0 0 0 1>}
 deform_upperarm_fr_base_l {parent=upperarm_twist_01_l, loc=<-5.6312 5.092 0.3354>, rot=<0 0 0 -1>}
 deform_upperarm_fr_l {parent=deform_upperarm_fr_base_l, loc=<0 0 0>, rot=<0 0 0 1>}
 deform_upperarm_up_base_l {parent=upperarm_twist_01_l, loc=<-8.891 0.4825 3.9517>, rot=<0 0 0 -1>}
 deform_upperarm_up_l {parent=deform_upperarm_up_base_l, loc=<0 0 0>, rot=<0 0 0 1>}
 clavicle_r {parent=spine_05, loc=<5.4831 -0.1053 3.2946>, rot=<0.7289 0.0465 -0.6829 0.009>}
 upperarm_r {parent=clavicle_r, loc=<-9.8658 1.2136 0.4105>, rot=<0.0164 -0.3245 -0.0269 -0.9454>}
 upperarm_twist_01_r {parent=upperarm_r, loc=<-7.2135 0.1134 0>, rot=<0 0 0 -1>}
 deform_upperarm_bk_base_r {parent=upperarm_twist_01_r, loc=<6.4053 4.3746 -0.314>, rot=<0 0 0 -1>}
 deform_upperarm_bk_r {parent=deform_upperarm_bk_base_r, loc=<0 0 0>, rot=<0 0 0 1>}
 deform_upperarm_fr_base_r {parent=upperarm_twist_01_r, loc=<5.6312 -5.092 -0.3354>, rot=<0 0 0 -1>}
 deform_upperarm_fr_r {parent=deform_upperarm_fr_base_r, loc=<0 0 0>, rot=<0 0 0 1>}
 deform_upperarm_up_base_r {parent=upperarm_twist_01_r, loc=<8.891 -0.4825 -3.9517>, rot=<0 0 0 -1>}
 deform_upperarm_up_r {parent=deform_upperarm_up_base_r, loc=<0 0 0>, rot=<0 0 0 1>}
 Hair l Aim {controlType=EULER_TRANSFORM, parent=spine_04, type=Control, loc=<25 -35 -10>, rot=<0 0 0 1>}
 Hair r Aim {controlType=EULER_TRANSFORM, parent=spine_04, type=Control, loc=<25 -35 10>, rot=<0 0 0 1>}
 dyn_hair_r_TGT {parent=pelvis, loc=<0.8236 -14.3959 10.3103>, rot=<-0.0502 -0.7053 0.0502 -0.7053>}
 dyn_hair_l_TGT {parent=pelvis, loc=<0.8235 -14.3959 -10.3108>, rot=<-0.0502 -0.7053 0.0502 -0.7053>}
]
