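# Physics asset for the F_MED_PencilFig female head's skydive hair, giving left and right hair chains their physics bodies.
# Each chain has six capsules that widen toward the tip, joined by swing- and twist-limited constraints and attached to the head through a tiny Look_at body.
# Head, neck, spine, clavicle and upper-arm capsules form the body volume, with angularly free constraints.
# The neck carries a second, much larger offset capsule.
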PhysicsAsset /BRCosmetics/Characters/Player/Female/Medium/Heads/F_MED_Pencil_Fig_Head/Meshes/Hair/F_MED_PencilFig_skydive_Hair_Physics
bodies {
 clavicle_l [{shapeType=Capsule, length=11.4794, radius=7.2519, center=<4.5 -0.3686 -2>, rotation=<rot 85.606 0 0>}]
 clavicle_r [{shapeType=Capsule, length=11.4794, radius=7.2519, center=<-4.5 0.3686 2>, rotation=<rot 85.606 0 0>}]
 dyn_hair_l_main_01 [{
   shapeType=Capsule
   length=2.4583
   radius=1.7588
   shapeName=dyn_hair_l_main_01_capsule
   center=<2.7417 0.3156 0.5484>
   rotation=<rot -20.5953 -90 29.4156>
  }]
 dyn_hair_l_main_02 [{shapeType=Capsule, length=3.25, radius=2.75, shapeName=dyn_hair_l_main_02_capsule, center=<4 0 0>, rotation=<rot 90 0 0>}]
 dyn_hair_l_main_03 [{shapeType=Capsule, length=6.25, radius=2.75, shapeName=dyn_hair_l_main_03_capsule, center=<5 0 0>, rotation=<rot 90 0 0>}]
 dyn_hair_l_main_04 [{shapeType=Capsule, length=5.25, radius=3.75, shapeName=dyn_hair_l_main_04_capsule, center=<5 0 0>, rotation=<rot 90 0 0>}]
 dyn_hair_l_main_05 [{shapeType=Capsule, length=5.25, radius=3.75, shapeName=dyn_hair_l_main_05_capsule, center=<5 0 0>, rotation=<rot 90 0 0>}]
 dyn_hair_l_main_06 [{shapeType=Capsule, length=5.25, radius=3.75, shapeName=dyn_hair_l_main_06_capsule, center=<5 0 0>, rotation=<rot 90 0 0>}]
 dyn_hair_l_main_Look_at [{shapeType=Capsule, length=0.505, radius=0.505, shapeName=dyn_hair_l_main_Look_at_capsule, rotation=<rot 0 -90 0>}]
 dyn_hair_r_main_01 [{
   shapeType=Capsule
   length=2.4583
   radius=1.7588
   shapeName=dyn_hair_r_main_01_capsule
   center=<2.7417 0.3156 -0.5484>
   rotation=<rot 20.5953 -90 -29.4156>
  }]
 dyn_hair_r_main_02 [{shapeType=Capsule, length=3.25, radius=2.75, shapeName=dyn_hair_r_main_02_capsule, center=<4 0 0>, rotation=<rot 90 112.0593 112.0592>}]
 dyn_hair_r_main_03 [{shapeType=Capsule, length=6.25, radius=2.75, shapeName=dyn_hair_r_main_03_capsule, center=<5 0 0>, rotation=<rot 90 96.0469 96.0468>}]
 dyn_hair_r_main_04 [{shapeType=Capsule, length=5.25, radius=3.75, shapeName=dyn_hair_r_main_04_capsule, center=<5 0 0>, rotation=<rot 90 -87.0235 -87.0237>}]
 dyn_hair_r_main_05 [{shapeType=Capsule, length=5.25, radius=3.75, shapeName=dyn_hair_r_main_05_capsule, center=<5 0 0>, rotation=<rot 90 -91.4368 -91.437>}]
 dyn_hair_r_main_06 [{shapeType=Capsule, length=5.25, radius=3.75, shapeName=dyn_hair_r_main_06_capsule, center=<5 0 0>, rotation=<rot 90 -89.8586 -89.8588>}]
 dyn_hair_r_main_Look_at [{shapeType=Capsule, length=0.505, radius=0.505, shapeName=dyn_hair_r_main_Look_at_capsule, rotation=<rot 0 -90 0>}]
 head [{shapeType=Capsule, length=4.5205, radius=7.9608, center=<5.3563 -0.043 -0.1056>, rotation=<rot 90 180 174.9638>}]
 neck_01 [
  {shapeType=Capsule, length=4.5, radius=5.5, rotation=<rot 90 0 10>}
  {shapeType=Capsule, length=50.75, radius=15.25, shapeName=neck_01_capsule, center=<1.929 13.6623 0>}
 ]
 spine_01 [{shapeType=Capsule, length=5.5613, radius=11.1791, center=<0 -0.3451 -0.0001>}]
 spine_02 [{shapeType=Capsule, length=3.7205, radius=10.505, center=<0 0.7078 -0.0001>}]
 spine_03 [{shapeType=Capsule, length=5.9414, radius=9.7882, center=<0 0.3989 -0.0001>}]
 spine_04 [{shapeType=Capsule, length=6.3263, radius=10.218, center=<0.6532 0.9298 0>}]
 spine_05 [{shapeType=Capsule, length=11.9068, radius=8.1314, center=<2.8332 0.6436 -0.0001>}]
 upperarm_l [{shapeType=Capsule, length=23.125, radius=6.0097, center=<15 -0.4936 0.3456>, rotation=<rot 90 0 0>}]
 upperarm_r [{shapeType=Capsule, length=23.125, radius=6.0097, center=<-15 0.4936 -0.3456>, rotation=<rot 90 0 0>}]
}
constraints [
 {
  bone1Name=upperarm_r
  bone2Name=clavicle_r
  swing1LimitDegrees=45
  swing1Motion=Free
  swing2LimitDegrees=45
  swing2Motion=Free
  twistLimitDegrees=45
  twistMotion=Free
 }
 {
  bone1Name=upperarm_l
  bone2Name=clavicle_l
  swing1LimitDegrees=45
  swing1Motion=Free
  swing2LimitDegrees=45
  swing2Motion=Free
  twistLimitDegrees=45
  twistMotion=Free
 }
 {
  bone1Name=head
  bone2Name=neck_01
  swing1LimitDegrees=45
  swing1Motion=Free
  swing2LimitDegrees=45
  swing2Motion=Free
  twistLimitDegrees=45
  twistMotion=Free
 }
 {
  bone1Name=clavicle_l
  bone2Name=spine_05
  swing1LimitDegrees=45
  swing1Motion=Free
  swing2LimitDegrees=45
  swing2Motion=Free
  twistLimitDegrees=45
  twistMotion=Free
 }
 {
  bone1Name=clavicle_r
  bone2Name=spine_05
  swing1LimitDegrees=45
  swing1Motion=Free
  swing2LimitDegrees=45
  swing2Motion=Free
  twistLimitDegrees=45
  twistMotion=Free
 }
 {
  bone1Name=spine_05
  bone2Name=spine_04
  swing1LimitDegrees=45
  swing1Motion=Free
  swing2LimitDegrees=45
  swing2Motion=Free
  twistLimitDegrees=45
  twistMotion=Free
 }
 {
  bone1Name=spine_02
  bone2Name=spine_01
  swing1LimitDegrees=45
  swing1Motion=Free
  swing2LimitDegrees=45
  swing2Motion=Free
  twistLimitDegrees=45
  twistMotion=Free
 }
 {
  bone1Name=spine_03
  bone2Name=spine_02
  swing1LimitDegrees=45
  swing1Motion=Free
  swing2LimitDegrees=45
  swing2Motion=Free
  twistLimitDegrees=45
  twistMotion=Free
 }
 {
  bone1Name=spine_04
  bone2Name=spine_03
  swing1LimitDegrees=45
  swing1Motion=Free
  swing2LimitDegrees=45
  swing2Motion=Free
  twistLimitDegrees=45
  twistMotion=Free
 }
 {
  bone1Name=neck_01
  bone2Name=spine_05
  swing1LimitDegrees=45
  swing1Motion=Free
  swing2LimitDegrees=45
  swing2Motion=Free
  twistLimitDegrees=45
  twistMotion=Free
 }
 {bone1Name=dyn_hair_l_main_02, bone2Name=dyn_hair_l_main_01, swing1LimitDegrees=35, swing2LimitDegrees=35, twistLimitDegrees=15}
 {bone1Name=dyn_hair_l_main_03, bone2Name=dyn_hair_l_main_02, swing1LimitDegrees=45, swing2LimitDegrees=45, twistLimitDegrees=15}
 {bone1Name=dyn_hair_l_main_04, bone2Name=dyn_hair_l_main_03, swing1LimitDegrees=45, swing2LimitDegrees=45, twistLimitDegrees=15}
 {bone1Name=dyn_hair_l_main_05, bone2Name=dyn_hair_l_main_04, swing1LimitDegrees=45, swing2LimitDegrees=45, twistLimitDegrees=15}
 {bone1Name=dyn_hair_l_main_06, bone2Name=dyn_hair_l_main_05, swing1LimitDegrees=45, swing2LimitDegrees=45, twistLimitDegrees=15}
 {bone1Name=dyn_hair_r_main_02, bone2Name=dyn_hair_r_main_01, swing1LimitDegrees=35, swing2LimitDegrees=35, twistLimitDegrees=15}
 {bone1Name=dyn_hair_r_main_03, bone2Name=dyn_hair_r_main_02, swing1LimitDegrees=45, swing2LimitDegrees=45, twistLimitDegrees=15}
 {bone1Name=dyn_hair_r_main_04, bone2Name=dyn_hair_r_main_03, swing1LimitDegrees=45, swing2LimitDegrees=45, twistLimitDegrees=15}
 {bone1Name=dyn_hair_r_main_05, bone2Name=dyn_hair_r_main_04, swing1LimitDegrees=45, swing2LimitDegrees=45, twistLimitDegrees=15}
 {bone1Name=dyn_hair_r_main_06, bone2Name=dyn_hair_r_main_05, swing1LimitDegrees=45, swing2LimitDegrees=45, twistLimitDegrees=15}
 {
  bone1Name=dyn_hair_r_main_01
  bone2Name=dyn_hair_r_main_Look_at
  swing1LimitDegrees=45
  swing1Motion=Free
  swing2LimitDegrees=45
  swing2Motion=Free
  twistLimitDegrees=45
  twistMotion=Free
 }
 {
  bone1Name=dyn_hair_r_main_Look_at
  bone2Name=head
  swing1LimitDegrees=45
  swing1Motion=Free
  swing2LimitDegrees=45
  swing2Motion=Free
  twistLimitDegrees=45
  twistMotion=Free
 }
 {
  bone1Name=dyn_hair_l_main_01
  bone2Name=dyn_hair_l_main_Look_at
  swing1LimitDegrees=45
  swing1Motion=Free
  swing2LimitDegrees=45
  swing2Motion=Free
  twistLimitDegrees=45
  twistMotion=Free
 }
 {
  bone1Name=dyn_hair_l_main_Look_at
  bone2Name=head
  swing1LimitDegrees=45
  swing1Motion=Free
  swing2LimitDegrees=45
  swing2Motion=Free
  twistLimitDegrees=45
  twistMotion=Free
 }
]
