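# Physics asset for the dangling earrings on the female medium Pencil Fig head cosmetic.
# Each ear has a three-bone capsule chain (dyn_ear) constrained from the head down, with swing limits and tighter twist limits.
# Capsules on the head, neck, upper spine and clavicles form the surrounding body shapes, linked by constraints with free swing and twist.
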PhysicsAsset /BRCosmetics/Characters/Player/Female/Medium/Heads/F_MED_Pencil_Fig_Head/Meshes/Hair/F_MED_Pencil_Fig_Earrings_Physics
bodies {
 clavicle_l [{shapeType=Capsule, length=5.1044, radius=7.2519, center=<4.5 -0.75 -2>, rotation=<rot 85.606 0 0>}]
 clavicle_r [{shapeType=Capsule, length=5.1044, radius=7.2519, center=<-4.5 0.75 2>, rotation=<rot 85.606 0 0>}]
 dyn_ear_l [{shapeType=Capsule, length=0.505, radius=0.505, shapeName=dyn_ear_l_capsule, center=<1.8324 -0.0028 -0.0029>, rotation=<rot 0 -90 -89.9654>}]
 dyn_ear_l_01 [{
   shapeType=Capsule
   length=4.755
   radius=1.255
   shapeName=dyn_ear_l_01_capsule
   center=<3.3585 0.0155 0.0003>
   rotation=<rot 0 90.055 -89.9823>
  }]
 dyn_ear_l_02 [{shapeType=Capsule, length=0.255, radius=0.755, shapeName=dyn_ear_l_02_capsule, center=<1.7141 -0.0157 0>, rotation=<rot 0 -89.9982 -89.9943>}]
 dyn_ear_r [{shapeType=Capsule, length=0.505, radius=0.505, shapeName=dyn_ear_r_capsule, center=<1.8324 -0.0028 0.0029>, rotation=<rot 0 90 -89.9654>}]
 dyn_ear_r_01 [{
   shapeType=Capsule
   length=4.755
   radius=1.255
   shapeName=dyn_ear_r_01_capsule
   center=<3.3585 0.0155 -0.0003>
   rotation=<rot 0 -89.945 -89.9823>
  }]
 dyn_ear_r_02 [{shapeType=Capsule, length=0.255, radius=0.755, shapeName=dyn_ear_r_02_capsule, center=<1.7141 -0.0157 0>, rotation=<rot 0 90.0018 -89.9943>}]
 head [{shapeType=Capsule, length=5.0205, radius=7.4608, center=<5.3563 -0.043 -0.1056>, rotation=<rot 90 180 174.9638>}]
 neck_01 [{shapeType=Capsule, length=4.875, radius=5.125, rotation=<rot 90 0 10>}]
 spine_05 [{shapeType=Capsule, length=3.2818, radius=8.1314, center=<4.5179 0.252 -0.0001>}]
}
constraints [
 {
  bone1Name=head
  bone2Name=neck_01
  swing1LimitDegrees=45
  swing1Motion=Free
  swing2LimitDegrees=45
  swing2Motion=Free
  twistLimitDegrees=45
  twistMotion=Free
 }
 {
  bone1Name=clavicle_l
  bone2Name=spine_05
  swing1LimitDegrees=45
  swing1Motion=Free
  swing2LimitDegrees=45
  swing2Motion=Free
  twistLimitDegrees=45
  twistMotion=Free
 }
 {
  bone1Name=clavicle_r
  bone2Name=spine_05
  swing1LimitDegrees=45
  swing1Motion=Free
  swing2LimitDegrees=45
  swing2Motion=Free
  twistLimitDegrees=45
  twistMotion=Free
 }
 {
  bone1Name=neck_01
  bone2Name=spine_05
  swing1LimitDegrees=45
  swing1Motion=Free
  swing2LimitDegrees=45
  swing2Motion=Free
  twistLimitDegrees=45
  twistMotion=Free
 }
 {bone1Name=dyn_ear_r_02, bone2Name=dyn_ear_r_01, swing1LimitDegrees=45, swing2LimitDegrees=45, twistLimitDegrees=15}
 {bone1Name=dyn_ear_l, bone2Name=head, swing1LimitDegrees=45, swing2LimitDegrees=45, twistLimitDegrees=15}
 {bone1Name=dyn_ear_l_01, bone2Name=dyn_ear_l, swing1LimitDegrees=45, swing2LimitDegrees=45, twistLimitDegrees=15}
 {bone1Name=dyn_ear_l_02, bone2Name=dyn_ear_l_01, swing1LimitDegrees=45, swing2LimitDegrees=45, twistLimitDegrees=15}
 {bone1Name=dyn_ear_r, bone2Name=head, swing1LimitDegrees=45, swing2LimitDegrees=45, twistLimitDegrees=15}
 {bone1Name=dyn_ear_r_01, bone2Name=dyn_ear_r, swing1LimitDegrees=45, swing2LimitDegrees=45, twistLimitDegrees=15}
]
