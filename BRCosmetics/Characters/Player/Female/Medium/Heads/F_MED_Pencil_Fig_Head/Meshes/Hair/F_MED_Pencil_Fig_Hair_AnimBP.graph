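# Hair AnimBlueprint for the Pencil Fig female head, adding secondary motion to hair and earrings.
# It copies the parent mesh's pose and includes a RigidBody simulation plus earring Anim Dynamics.
# A bRigidBodyEnabled blend chooses between the cached 'Perfmode' pose and that pose with main hair chains constrained toward spine and clavicles.
# Strand bones then follow the main chains through constraints, layered with sprung Anim Dynamics using spherical body limits.
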AnimBlueprint /BRCosmetics/Characters/Player/Female/Medium/Heads/F_MED_Pencil_Fig_Head/Meshes/Hair/F_MED_Pencil_Fig_Hair_AnimBP
target_skeleton=/BRCosmetics/Characters/Player/Common/Fortnite_Base_Head/Pencil/Fortnite_Base_Head_Pencil_Skeleton
graphs [
 AnimGraph {nodes [
   n1 = Anim Dynamics - Bone: dyn_earring_r(Alpha=1, ExternalForce="0,0,0") {
    angularDampingOverride=1
    angularSpringConstant=5000
    bAngularSpring=true
    bOverrideAngularDamping=true
    bUseGravityOverride=true
    gravityScale=0.35
    simulationSpace=RootRelative
    alphaBoolBlend {blendOption=Linear}
    boundBone {boneName=dyn_earring_r}
    componentLinearAccScale=<1 1 1>
    componentLinearVelScale=<1 1 1>
    gravityOverride {z=-980}
    physicsBodyDefinitions [{
      sphereCollisionRadius=0
      boundBone {boneName=dyn_earring_r}
      boxExtents=<50 60 60>
      constraintSetup {angularLimitsMax=<5 30 30>, angularLimitsMin=<-5 -30 -30>, angularTarget=<1 0 0>}
      localJointOffset=<4 0 0>
     }]
   }
   Anim Dynamics - Bone: dyn_earring_l(ComponentPose=@n1, Alpha=1, ExternalForce="0,0,0") {
    angularDampingOverride=1
    angularSpringConstant=5000
    bAngularSpring=true
    bOverrideAngularDamping=true
    bUseGravityOverride=true
    gravityScale=0.35
    simulationSpace=RootRelative
    alphaBoolBlend {blendOption=Linear}
    boundBone {boneName=dyn_earring_l}
    componentLinearAccScale=<1 1 1>
    componentLinearVelScale=<1 1 1>
    gravityOverride {z=-980}
    physicsBodyDefinitions [{
      sphereCollisionRadius=0
      boundBone {boneName=dyn_earring_l}
      boxExtents=<50 60 60>
      constraintSetup {angularLimitsMax=<5 30 30>, angularLimitsMin=<-5 -30 -30>, angularTarget=<1 0 0>}
      localJointOffset=<4 0 0>
     }]
   }
   Anim Dynamics - Bone: dyn_hair_r_a_01(Alpha=0.6, ExternalForce="0,0,0") {
    angularDampingOverride=0.98
    angularSpringConstant=150
    bAngularSpring=true
    bGravityOverrideInSimSpace=true
    bOverrideAngularDamping=true
    simulationSpace=RootRelative
    alphaBoolBlend {blendOption=Linear}
    boundBone {boneName=dyn_hair_r_a_01}
    componentAppliedLinearAccClamp=<1500 1500 1500>
    componentLinearAccScale=<0.6 0.6 0.6>
    componentLinearVelScale=<0.6 0.6 0.6>
    physicsBodyDefinitions [{
      sphereCollisionRadius=0
      boundBone {boneName=dyn_hair_r_a_01}
      boxExtents=<12 15 15>
      constraintSetup {angularLimitsMax=<1 5 0>, angularLimitsMin=<-1 -5 -15>, angularTarget=<1 0 0>}
      localJointOffset=<1 0 0>
     }]
    relativeSpaceBone {boneName=root}
   }
   Copy Pose From Mesh() {bCopyCurves=true, bUseAttachedParent=true}
   RigidBody!(ComponentPose, Alpha, OverrideWorldGravity:Vector="0,0,-980", ExternalForce) -> (Pose) {
    bClampLinearTranslationLimitToRefPose=true
    bOverrideWorldGravity=true
    overridePhysicsAsset=&/F_MED_Pencil_Fig_Earrings_Physics
    simulationSpace=BaseBoneSpace
    alphaBoolBlend {blendOption=Linear}
    baseBoneRef {boneName=spine_05}
    componentAppliedLinearAccClamp=<2500 2500 2500>
    componentLinearAccScale=<0.35 0.35 0.5>
    componentLinearVelScale=<0.35 0.35 0.5>
    overrideWorldGravity {z=-800}
   }
   n2 = "Constraint - dyn_hair_l_main_03 (PP)"!(ComponentPose=Local To Component(LocalPose=Use cached pose 'Perfmode'()), Alpha=0.5, ConstraintWeights_0=0.3, ConstraintWeights_1:float=0.7) -> (Pose) {
    alphaBoolBlend {blendOption=Linear}
    boneToModify {boneName=dyn_hair_l_main_03}
    constraintSetup [{targetBone {boneName=spine_05}}, {targetBone {boneName=clavicle_l}}]
    constraintWeights [1,1]
   }
   n3 = "Constraint - dyn_hair_l_main_04 (PP)"!(ComponentPose=@n2, Alpha=0.7, ConstraintWeights_0=0.4, ConstraintWeights_1:float=0.6) -> (Pose) {
    alphaBoolBlend {blendOption=Linear}
    boneToModify {boneName=dyn_hair_l_main_04}
    constraintSetup [{targetBone {boneName=spine_05}}, {targetBone {boneName=clavicle_l}}]
    constraintWeights [1,1]
   }
   n4 = "Constraint - dyn_hair_l_main_05 (PP)"!(ComponentPose=@n3, Alpha=1, ConstraintWeights_0=0.7, ConstraintWeights_1:float=0.3) -> (Pose) {
    alphaBoolBlend {blendOption=Linear}
    boneToModify {boneName=dyn_hair_l_main_05}
    constraintSetup [{targetBone {boneName=spine_05}}, {targetBone {boneName=spine_04}}]
    constraintWeights [1,1]
   }
   n5 = "Constraint - dyn_hair_l_main_06 (PP)"!(ComponentPose=@n4, Alpha=1, ConstraintWeights_0=0.5, ConstraintWeights_1:float=0.5) -> (Pose) {
    alphaBoolBlend {blendOption=Linear}
    boneToModify {boneName=dyn_hair_l_main_06}
    constraintSetup [{targetBone {boneName=spine_05}}, {targetBone {boneName=spine_04}}]
    constraintWeights [1,1]
   }
   n6 = "Constraint - dyn_hair_r_main_03 (PP)"!(ComponentPose=@n5, Alpha=0.3, ConstraintWeights_0=0.3, ConstraintWeights_1:float=0.7) -> (Pose) {
    alphaBoolBlend {blendOption=Linear}
    boneToModify {boneName=dyn_hair_r_main_03}
    constraintSetup [{targetBone {boneName=spine_05}}, {targetBone {boneName=clavicle_r}}]
    constraintWeights [1,1]
   }
   n7 = "Constraint - dyn_hair_r_main_04 (PP)"!(ComponentPose=@n6, Alpha=0.5, ConstraintWeights_0=0.5, ConstraintWeights_1:float=0.5) -> (Pose) {
    alphaBoolBlend {blendOption=Linear}
    boneToModify {boneName=dyn_hair_r_main_04}
    constraintSetup [{targetBone {boneName=spine_05}}, {targetBone {boneName=clavicle_r}}]
    constraintWeights [1,1]
   }
   n8 = "Constraint - dyn_hair_r_main_05 (PP)"!(ComponentPose=@n7, Alpha=1, ConstraintWeights_0=0.7, ConstraintWeights_1:float=0.3) -> (Pose) {
    alphaBoolBlend {blendOption=Linear}
    boneToModify {boneName=dyn_hair_r_main_05}
    constraintSetup [{targetBone {boneName=spine_05}}, {targetBone {boneName=spine_04}}]
    constraintWeights [1,1]
   }
   n9 = "Constraint - dyn_hair_r_main_06 (PP)"!(ComponentPose=@n8, Alpha=1, ConstraintWeights_0=0.5, ConstraintWeights_1:float=0.5) -> (Pose) {
    alphaBoolBlend {blendOption=Linear}
    boneToModify {boneName=dyn_hair_r_main_06}
    constraintSetup [{targetBone {boneName=spine_05}}, {targetBone {boneName=spine_04}}]
    constraintWeights [1,1]
   }
   n10 = Component To Local(ComponentPose=@n9)
   n11 = Use cached pose 'Perfmode'()
   n12 = Get bRigidBodyEnabled()
   n13 = "Constraint - dyn_hair_r_a_01 (P)"(
    ComponentPose=Local To Component(LocalPose=Blend Poses by bool(BlendPose_0=@n11, BlendPose_1=@n10, bActiveValue=@n12) {blendType=Linear})
    Alpha=1
    ConstraintWeights_0=1
   ) {
    alphaBoolBlend {blendOption=Linear}
    boneToModify {boneName=dyn_hair_r_a_01}
    constraintSetup [{targetBone {boneName=dyn_hair_r_main_02}}]
    constraintWeights [1]
   }
   n14 = "Constraint - dyn_hair_r_a_02 (P)"(ComponentPose=@n13, Alpha=1, ConstraintWeights_0=1) {
    alphaBoolBlend {blendOption=Linear}
    boneToModify {boneName=dyn_hair_r_a_02}
    constraintSetup [{targetBone {boneName=dyn_hair_r_main_03}}]
    constraintWeights [1]
   }
   n15 = "Constraint - dyn_hair_r_a_03 (P)"(ComponentPose=@n14, Alpha=1, ConstraintWeights_0=1) {
    alphaBoolBlend {blendOption=Linear}
    boneToModify {boneName=dyn_hair_r_a_03}
    constraintSetup [{targetBone {boneName=dyn_hair_r_main_04}}]
    constraintWeights [1]
   }
   n16 = "Constraint - dyn_hair_r_a_04 (P)"(ComponentPose=@n15, Alpha=1, ConstraintWeights_0=1) {
    alphaBoolBlend {blendOption=Linear}
    boneToModify {boneName=dyn_hair_r_a_04}
    constraintSetup [{targetBone {boneName=dyn_hair_r_main_05}}]
    constraintWeights [1]
   }
   n17 = "Constraint - dyn_hair_r_a_05 (P)"(ComponentPose=@n16, Alpha=1, ConstraintWeights_0=1) {
    alphaBoolBlend {blendOption=Linear}
    boneToModify {boneName=dyn_hair_r_a_05}
    constraintSetup [{targetBone {boneName=dyn_hair_r_main_06}}]
    constraintWeights [1]
   }
   n18 = "Constraint - dyn_hair_r_b_01 (P)"(ComponentPose=@n17, Alpha=1, ConstraintWeights_0=1) {
    alphaBoolBlend {blendOption=Linear}
    boneToModify {boneName=dyn_hair_r_b_01}
    constraintSetup [{targetBone {boneName=dyn_hair_r_main_02}}]
    constraintWeights [1]
   }
   n19 = "Constraint - dyn_hair_r_b_02 (P)"(ComponentPose=@n18, Alpha=1, ConstraintWeights_0=1) {
    alphaBoolBlend {blendOption=Linear}
    boneToModify {boneName=dyn_hair_r_b_02}
    constraintSetup [{targetBone {boneName=dyn_hair_r_main_03}}]
    constraintWeights [1]
   }
   n20 = "Constraint - dyn_hair_r_b_03 (P)"(ComponentPose=@n19, Alpha=1, ConstraintWeights_0=1) {
    alphaBoolBlend {blendOption=Linear}
    boneToModify {boneName=dyn_hair_r_b_03}
    constraintSetup [{targetBone {boneName=dyn_hair_r_main_04}}]
    constraintWeights [1]
   }
   n21 = "Constraint - dyn_hair_r_b_04 (P)"(ComponentPose=@n20, Alpha=1, ConstraintWeights_0=1) {
    alphaBoolBlend {blendOption=Linear}
    boneToModify {boneName=dyn_hair_r_b_04}
    constraintSetup [{targetBone {boneName=dyn_hair_r_main_05}}]
    constraintWeights [1]
   }
   n22 = "Constraint - dyn_hair_r_b_05 (P)"(ComponentPose=@n21, Alpha=1, ConstraintWeights_0=1) {
    alphaBoolBlend {blendOption=Linear}
    boneToModify {boneName=dyn_hair_r_b_05}
    constraintSetup [{targetBone {boneName=dyn_hair_r_main_06}}]
    constraintWeights [1]
   }
   n23 = "Constraint - dyn_hair_r_c_01 (P)"(ComponentPose=@n22, Alpha=1, ConstraintWeights_0=1) {
    alphaBoolBlend {blendOption=Linear}
    boneToModify {boneName=dyn_hair_r_c_01}
    constraintSetup [{targetBone {boneName=dyn_hair_r_main_02}}]
    constraintWeights [1]
   }
   n24 = "Constraint - dyn_hair_r_c_02 (P)"(ComponentPose=@n23, Alpha=1, ConstraintWeights_0=1) {
    alphaBoolBlend {blendOption=Linear}
    boneToModify {boneName=dyn_hair_r_c_02}
    constraintSetup [{targetBone {boneName=dyn_hair_r_main_03}}]
    constraintWeights [1]
   }
   n25 = "Constraint - dyn_hair_r_c_03 (P)"(ComponentPose=@n24, Alpha=1, ConstraintWeights_0=1) {
    alphaBoolBlend {blendOption=Linear}
    boneToModify {boneName=dyn_hair_r_c_03}
    constraintSetup [{targetBone {boneName=dyn_hair_r_main_04}}]
    constraintWeights [1]
   }
   n26 = "Constraint - dyn_hair_r_c_04 (P)"(ComponentPose=@n25, Alpha=1, ConstraintWeights_0=1) {
    alphaBoolBlend {blendOption=Linear}
    boneToModify {boneName=dyn_hair_r_c_04}
    constraintSetup [{targetBone {boneName=dyn_hair_r_main_05}}]
    constraintWeights [1]
   }
   n27 = "Constraint - dyn_hair_r_c_05 (P)"(ComponentPose=@n26, Alpha=1, ConstraintWeights_0=1) {
    alphaBoolBlend {blendOption=Linear}
    boneToModify {boneName=dyn_hair_r_c_05}
    constraintSetup [{targetBone {boneName=dyn_hair_r_main_06}}]
    constraintWeights [1]
   }
   n28 = "Constraint - dyn_hair_r_d_01 (P)"(ComponentPose=@n27, Alpha=1, ConstraintWeights_0=1) {
    alphaBoolBlend {blendOption=Linear}
    boneToModify {boneName=dyn_hair_r_d_01}
    constraintSetup [{targetBone {boneName=dyn_hair_r_main_02}}]
    constraintWeights [1]
   }
   n29 = "Constraint - dyn_hair_r_d_02 (P)"(ComponentPose=@n28, Alpha=1, ConstraintWeights_0=1) {
    alphaBoolBlend {blendOption=Linear}
    boneToModify {boneName=dyn_hair_r_d_02}
    constraintSetup [{targetBone {boneName=dyn_hair_r_main_03}}]
    constraintWeights [1]
   }
   n30 = "Constraint - dyn_hair_r_d_03 (P)"(ComponentPose=@n29, Alpha=1, ConstraintWeights_0=1) {
    alphaBoolBlend {blendOption=Linear}
    boneToModify {boneName=dyn_hair_r_d_03}
    constraintSetup [{targetBone {boneName=dyn_hair_r_main_04}}]
    constraintWeights [1]
   }
   n31 = "Constraint - dyn_hair_r_d_04 (P)"(ComponentPose=@n30, Alpha=1, ConstraintWeights_0=1) {
    alphaBoolBlend {blendOption=Linear}
    boneToModify {boneName=dyn_hair_r_d_04}
    constraintSetup [{targetBone {boneName=dyn_hair_r_main_05}}]
    constraintWeights [1]
   }
   n32 = "Constraint - dyn_hair_r_d_05 (P)"(ComponentPose=@n31, Alpha=1, ConstraintWeights_0=1) {
    alphaBoolBlend {blendOption=Linear}
    boneToModify {boneName=dyn_hair_r_d_05}
    constraintSetup [{targetBone {boneName=dyn_hair_r_main_06}}]
    constraintWeights [1]
   }
   n33 = "Constraint - dyn_hair_r_e_01 (P)"(ComponentPose=@n32, Alpha=1, ConstraintWeights_0=1) {
    alphaBoolBlend {blendOption=Linear}
    boneToModify {boneName=dyn_hair_r_e_01}
    constraintSetup [{targetBone {boneName=dyn_hair_r_main_02}}]
    constraintWeights [1]
   }
   n34 = "Constraint - dyn_hair_r_e_02 (P)"(ComponentPose=@n33, Alpha=1, ConstraintWeights_0=1) {
    alphaBoolBlend {blendOption=Linear}
    boneToModify {boneName=dyn_hair_r_e_02}
    constraintSetup [{targetBone {boneName=dyn_hair_r_main_03}}]
    constraintWeights [1]
   }
   n35 = "Constraint - dyn_hair_r_e_03 (P)"(ComponentPose=@n34, Alpha=1, ConstraintWeights_0=1) {
    alphaBoolBlend {blendOption=Linear}
    boneToModify {boneName=dyn_hair_r_e_03}
    constraintSetup [{targetBone {boneName=dyn_hair_r_main_04}}]
    constraintWeights [1]
   }
   n36 = "Constraint - dyn_hair_r_e_04 (P)"(ComponentPose=@n35, Alpha=1, ConstraintWeights_0=1) {
    alphaBoolBlend {blendOption=Linear}
    boneToModify {boneName=dyn_hair_r_e_04}
    constraintSetup [{targetBone {boneName=dyn_hair_r_main_05}}]
    constraintWeights [1]
   }
   n37 = "Constraint - dyn_hair_r_e_05 (P)"(ComponentPose=@n36, Alpha=1, ConstraintWeights_0=1) {
    alphaBoolBlend {blendOption=Linear}
    boneToModify {boneName=dyn_hair_r_e_05}
    constraintSetup [{targetBone {boneName=dyn_hair_r_main_06}}]
    constraintWeights [1]
   }
   n38 = "Constraint - dyn_hair_r_f_01 (P)"(ComponentPose=@n37, Alpha=1, ConstraintWeights_0=1) {
    alphaBoolBlend {blendOption=Linear}
    boneToModify {boneName=dyn_hair_r_f_01}
    constraintSetup [{targetBone {boneName=dyn_hair_r_main_02}}]
    constraintWeights [1]
   }
   n39 = "Constraint - dyn_hair_r_f_02 (P)"(ComponentPose=@n38, Alpha=1, ConstraintWeights_0=1) {
    alphaBoolBlend {blendOption=Linear}
    boneToModify {boneName=dyn_hair_r_f_02}
    constraintSetup [{targetBone {boneName=dyn_hair_r_main_03}}]
    constraintWeights [1]
   }
   n40 = "Constraint - dyn_hair_r_f_03 (P)"(ComponentPose=@n39, Alpha=1, ConstraintWeights_0=1) {
    alphaBoolBlend {blendOption=Linear}
    boneToModify {boneName=dyn_hair_r_f_03}
    constraintSetup [{targetBone {boneName=dyn_hair_r_main_04}}]
    constraintWeights [1]
   }
   n41 = "Constraint - dyn_hair_r_f_04 (P)"(ComponentPose=@n40, Alpha=1, ConstraintWeights_0=1) {
    alphaBoolBlend {blendOption=Linear}
    boneToModify {boneName=dyn_hair_r_f_04}
    constraintSetup [{targetBone {boneName=dyn_hair_r_main_05}}]
    constraintWeights [1]
   }
   n42 = "Constraint - dyn_hair_r_f_05 (P)"(ComponentPose=@n41, Alpha=1, ConstraintWeights_0=1) {
    alphaBoolBlend {blendOption=Linear}
    boneToModify {boneName=dyn_hair_r_f_05}
    constraintSetup [{targetBone {boneName=dyn_hair_r_main_06}}]
    constraintWeights [1]
   }
   n43 = "Constraint - dyn_hair_l_a_01 (P)"(ComponentPose=@n42, Alpha=1, ConstraintWeights_0=1) {
    alphaBoolBlend {blendOption=Linear}
    boneToModify {boneName=dyn_hair_l_a_01}
    constraintSetup [{targetBone {boneName=dyn_hair_l_main_02}}]
    constraintWeights [1]
   }
   n44 = "Constraint - dyn_hair_l_a_02 (P)"(ComponentPose=@n43, Alpha=1, ConstraintWeights_0=1) {
    alphaBoolBlend {blendOption=Linear}
    boneToModify {boneName=dyn_hair_l_a_02}
    constraintSetup [{targetBone {boneName=dyn_hair_l_main_03}}]
    constraintWeights [1]
   }
   n45 = "Constraint - dyn_hair_l_a_03 (P)"(ComponentPose=@n44, Alpha=1, ConstraintWeights_0=1) {
    alphaBoolBlend {blendOption=Linear}
    boneToModify {boneName=dyn_hair_l_a_03}
    constraintSetup [{targetBone {boneName=dyn_hair_l_main_04}}]
    constraintWeights [1]
   }
   n46 = "Constraint - dyn_hair_l_a_04 (P)"(ComponentPose=@n45, Alpha=1, ConstraintWeights_0=1) {
    alphaBoolBlend {blendOption=Linear}
    boneToModify {boneName=dyn_hair_l_a_04}
    constraintSetup [{targetBone {boneName=dyn_hair_l_main_05}}]
    constraintWeights [1]
   }
   n47 = "Constraint - dyn_hair_l_a_05 (P)"(ComponentPose=@n46, Alpha=1, ConstraintWeights_0=1) {
    alphaBoolBlend {blendOption=Linear}
    boneToModify {boneName=dyn_hair_l_a_05}
    constraintSetup [{targetBone {boneName=dyn_hair_l_main_06}}]
    constraintWeights [1]
   }
   n48 = "Constraint - dyn_hair_l_b_01 (P)"(ComponentPose=@n47, Alpha=1, ConstraintWeights_0=1) {
    alphaBoolBlend {blendOption=Linear}
    boneToModify {boneName=dyn_hair_l_b_01}
    constraintSetup [{targetBone {boneName=dyn_hair_l_main_02}}]
    constraintWeights [1]
   }
   n49 = "Constraint - dyn_hair_l_b_02 (P)"(ComponentPose=@n48, Alpha=1, ConstraintWeights_0=1) {
    alphaBoolBlend {blendOption=Linear}
    boneToModify {boneName=dyn_hair_l_b_02}
    constraintSetup [{targetBone {boneName=dyn_hair_l_main_03}}]
    constraintWeights [1]
   }
   n50 = "Constraint - dyn_hair_l_b_03 (P)"(ComponentPose=@n49, Alpha=1, ConstraintWeights_0=1) {
    alphaBoolBlend {blendOption=Linear}
    boneToModify {boneName=dyn_hair_l_b_03}
    constraintSetup [{targetBone {boneName=dyn_hair_l_main_04}}]
    constraintWeights [1]
   }
   n51 = "Constraint - dyn_hair_l_b_04 (P)"(ComponentPose=@n50, Alpha=1, ConstraintWeights_0=1) {
    alphaBoolBlend {blendOption=Linear}
    boneToModify {boneName=dyn_hair_l_b_04}
    constraintSetup [{targetBone {boneName=dyn_hair_l_main_05}}]
    constraintWeights [1]
   }
   n52 = "Constraint - dyn_hair_l_b_05 (P)"(ComponentPose=@n51, Alpha=1, ConstraintWeights_0=1) {
    alphaBoolBlend {blendOption=Linear}
    boneToModify {boneName=dyn_hair_l_b_05}
    constraintSetup [{targetBone {boneName=dyn_hair_l_main_06}}]
    constraintWeights [1]
   }
   n53 = "Constraint - dyn_hair_l_c_01 (P)"(ComponentPose=@n52, Alpha=1, ConstraintWeights_0=1) {
    alphaBoolBlend {blendOption=Linear}
    boneToModify {boneName=dyn_hair_l_c_01}
    constraintSetup [{targetBone {boneName=dyn_hair_l_main_02}}]
    constraintWeights [1]
   }
   n54 = "Constraint - dyn_hair_l_c_02 (P)"(ComponentPose=@n53, Alpha=1, ConstraintWeights_0=1) {
    alphaBoolBlend {blendOption=Linear}
    boneToModify {boneName=dyn_hair_l_c_02}
    constraintSetup [{targetBone {boneName=dyn_hair_l_main_03}}]
    constraintWeights [1]
   }
   n55 = "Constraint - dyn_hair_l_c_03 (P)"(ComponentPose=@n54, Alpha=1, ConstraintWeights_0=1) {
    alphaBoolBlend {blendOption=Linear}
    boneToModify {boneName=dyn_hair_l_c_03}
    constraintSetup [{targetBone {boneName=dyn_hair_l_main_04}}]
    constraintWeights [1]
   }
   n56 = "Constraint - dyn_hair_l_c_04 (P)"(ComponentPose=@n55, Alpha=1, ConstraintWeights_0=1) {
    alphaBoolBlend {blendOption=Linear}
    boneToModify {boneName=dyn_hair_l_c_04}
    constraintSetup [{targetBone {boneName=dyn_hair_l_main_05}}]
    constraintWeights [1]
   }
   n57 = "Constraint - dyn_hair_l_c_05 (P)"(ComponentPose=@n56, Alpha=1, ConstraintWeights_0=1) {
    alphaBoolBlend {blendOption=Linear}
    boneToModify {boneName=dyn_hair_l_c_05}
    constraintSetup [{targetBone {boneName=dyn_hair_l_main_06}}]
    constraintWeights [1]
   }
   n58 = "Constraint - dyn_hair_l_d_01 (P)"(ComponentPose=@n57, Alpha=1, ConstraintWeights_0=1) {
    alphaBoolBlend {blendOption=Linear}
    boneToModify {boneName=dyn_hair_l_d_01}
    constraintSetup [{targetBone {boneName=dyn_hair_l_main_02}}]
    constraintWeights [1]
   }
   n59 = "Constraint - dyn_hair_l_d_02 (P)"(ComponentPose=@n58, Alpha=1, ConstraintWeights_0=1) {
    alphaBoolBlend {blendOption=Linear}
    boneToModify {boneName=dyn_hair_l_d_02}
    constraintSetup [{targetBone {boneName=dyn_hair_l_main_03}}]
    constraintWeights [1]
   }
   n60 = "Constraint - dyn_hair_l_d_03 (P)"(ComponentPose=@n59, Alpha=1, ConstraintWeights_0=1) {
    alphaBoolBlend {blendOption=Linear}
    boneToModify {boneName=dyn_hair_l_d_03}
    constraintSetup [{targetBone {boneName=dyn_hair_l_main_04}}]
    constraintWeights [1]
   }
   n61 = "Constraint - dyn_hair_l_d_04 (P)"(ComponentPose=@n60, Alpha=1, ConstraintWeights_0=1) {
    alphaBoolBlend {blendOption=Linear}
    boneToModify {boneName=dyn_hair_l_d_04}
    constraintSetup [{targetBone {boneName=dyn_hair_l_main_05}}]
    constraintWeights [1]
   }
   n62 = "Constraint - dyn_hair_l_d_05 (P)"(ComponentPose=@n61, Alpha=1, ConstraintWeights_0=1) {
    alphaBoolBlend {blendOption=Linear}
    boneToModify {boneName=dyn_hair_l_d_05}
    constraintSetup [{targetBone {boneName=dyn_hair_l_main_06}}]
    constraintWeights [1]
   }
   n63 = "Constraint - dyn_hair_l_e_01 (P)"(ComponentPose=@n62, Alpha=1, ConstraintWeights_0=1) {
    alphaBoolBlend {blendOption=Linear}
    boneToModify {boneName=dyn_hair_l_e_01}
    constraintSetup [{targetBone {boneName=dyn_hair_l_main_02}}]
    constraintWeights [1]
   }
   n64 = "Constraint - dyn_hair_l_e_02 (P)"(ComponentPose=@n63, Alpha=1, ConstraintWeights_0=1) {
    alphaBoolBlend {blendOption=Linear}
    boneToModify {boneName=dyn_hair_l_e_02}
    constraintSetup [{targetBone {boneName=dyn_hair_l_main_03}}]
    constraintWeights [1]
   }
   n65 = "Constraint - dyn_hair_l_e_03 (P)"(ComponentPose=@n64, Alpha=1, ConstraintWeights_0=1) {
    alphaBoolBlend {blendOption=Linear}
    boneToModify {boneName=dyn_hair_l_e_03}
    constraintSetup [{targetBone {boneName=dyn_hair_l_main_04}}]
    constraintWeights [1]
   }
   n66 = "Constraint - dyn_hair_l_e_04 (P)"(ComponentPose=@n65, Alpha=1, ConstraintWeights_0=1) {
    alphaBoolBlend {blendOption=Linear}
    boneToModify {boneName=dyn_hair_l_e_04}
    constraintSetup [{targetBone {boneName=dyn_hair_l_main_05}}]
    constraintWeights [1]
   }
   n67 = "Constraint - dyn_hair_l_e_05 (P)"(ComponentPose=@n66, Alpha=1, ConstraintWeights_0=1) {
    alphaBoolBlend {blendOption=Linear}
    boneToModify {boneName=dyn_hair_l_e_05}
    constraintSetup [{targetBone {boneName=dyn_hair_l_main_06}}]
    constraintWeights [1]
   }
   n68 = "Constraint - dyn_hair_l_f_01 (P)"(ComponentPose=@n67, Alpha=1, ConstraintWeights_0=1) {
    alphaBoolBlend {blendOption=Linear}
    boneToModify {boneName=dyn_hair_l_f_01}
    constraintSetup [{targetBone {boneName=dyn_hair_l_main_02}}]
    constraintWeights [1]
   }
   n69 = "Constraint - dyn_hair_l_f_02 (P)"(ComponentPose=@n68, Alpha=1, ConstraintWeights_0=1) {
    alphaBoolBlend {blendOption=Linear}
    boneToModify {boneName=dyn_hair_l_f_02}
    constraintSetup [{targetBone {boneName=dyn_hair_l_main_03}}]
    constraintWeights [1]
   }
   n70 = "Constraint - dyn_hair_l_f_03 (P)"(ComponentPose=@n69, Alpha=1, ConstraintWeights_0=1) {
    alphaBoolBlend {blendOption=Linear}
    boneToModify {boneName=dyn_hair_l_f_03}
    constraintSetup [{targetBone {boneName=dyn_hair_l_main_04}}]
    constraintWeights [1]
   }
   n71 = "Constraint - dyn_hair_l_f_04 (P)"(ComponentPose=@n70, Alpha=1, ConstraintWeights_0=1) {
    alphaBoolBlend {blendOption=Linear}
    boneToModify {boneName=dyn_hair_l_f_04}
    constraintSetup [{targetBone {boneName=dyn_hair_l_main_05}}]
    constraintWeights [1]
   }
   n72 = "Constraint - dyn_hair_l_f_05 (P)"(ComponentPose=@n71, Alpha=1, ConstraintWeights_0=1) {
    alphaBoolBlend {blendOption=Linear}
    boneToModify {boneName=dyn_hair_l_f_05}
    constraintSetup [{targetBone {boneName=dyn_hair_l_main_06}}]
    constraintWeights [1]
   }
   n73 = Anim Dynamics - Bone: dyn_hair_r_a_02(ComponentPose=@n72, Alpha=0.5, ExternalForce="0,0,0") {
    angularDampingOverride=0.98
    angularSpringConstant=150
    bAngularSpring=true
    bGravityOverrideInSimSpace=true
    bOverrideAngularDamping=true
    simulationSpace=RootRelative
    alphaBoolBlend {blendOption=Linear}
    boundBone {boneName=dyn_hair_r_a_02}
    componentAppliedLinearAccClamp=<1500 1500 1500>
    componentLinearAccScale=<0.6 0.6 0.6>
    componentLinearVelScale=<0.6 0.6 0.6>
    physicsBodyDefinitions [{
      sphereCollisionRadius=0
      boundBone {boneName=dyn_hair_r_a_02}
      boxExtents=<12 15 15>
      constraintSetup {angularLimitsMax=<1 5 0>, angularLimitsMin=<-1 -5 -15>, angularTarget=<1 0 0>}
      localJointOffset=<1 0 0>
     }]
    relativeSpaceBone {boneName=root}
   }
   n74 = Anim Dynamics - Bone: dyn_hair_r_a_03(ComponentPose=@n73, Alpha=0.5, ExternalForce="0,0,0") {
    angularDampingOverride=0.98
    angularSpringConstant=150
    bAngularSpring=true
    bGravityOverrideInSimSpace=true
    bOverrideAngularDamping=true
    simulationSpace=RootRelative
    alphaBoolBlend {blendOption=Linear}
    boundBone {boneName=dyn_hair_r_a_03}
    componentAppliedLinearAccClamp=<1500 1500 1500>
    componentLinearAccScale=<0.6 0.6 0.6>
    componentLinearVelScale=<0.6 0.6 0.6>
    physicsBodyDefinitions [{
      sphereCollisionRadius=0
      boundBone {boneName=dyn_hair_r_a_03}
      boxExtents=<12 15 15>
      constraintSetup {angularLimitsMax=<1 5 0>, angularLimitsMin=<-1 -5 -15>, angularTarget=<1 0 0>}
      localJointOffset=<1 0 0>
     }]
    relativeSpaceBone {boneName=root}
   }
   n75 = Anim Dynamics - Bone: dyn_hair_r_a_04(ComponentPose=@n74, Alpha=0.5, ExternalForce="0,0,0") {
    angularDampingOverride=0.98
    angularSpringConstant=150
    bAngularSpring=true
    bGravityOverrideInSimSpace=true
    bOverrideAngularDamping=true
    bUseSphericalLimits=true
    simulationSpace=RootRelative
    alphaBoolBlend {blendOption=Linear}
    boundBone {boneName=dyn_hair_r_a_04}
    componentAppliedLinearAccClamp=<1500 1500 1500>
    componentLinearAccScale=<0.6 0.6 0.6>
    componentLinearVelScale=<0.6 0.6 0.6>
    physicsBodyDefinitions [{
      sphereCollisionRadius=0
      boundBone {boneName=dyn_hair_r_a_04}
      boxExtents=<12 15 15>
      constraintSetup {angularLimitsMax=<1 5 0>, angularLimitsMin=<-1 -5 -15>, angularTarget=<1 0 0>}
      localJointOffset=<1 0 0>
     }]
    relativeSpaceBone {boneName=root}
    sphericalLimits [
     {limitRadius=7, drivingBone {boneName=upperarm_r}, sphereLocalOffset=<-15 0 0>}
     {limitRadius=15, drivingBone {boneName=spine_04}, sphereLocalOffset=<0 5 0>}
    ]
   }
   n76 = Anim Dynamics - Bone: dyn_hair_r_a_05(ComponentPose=@n75, Alpha=0.5, ExternalForce="0,0,0") {
    angularDampingOverride=0.98
    angularSpringConstant=150
    bAngularSpring=true
    bGravityOverrideInSimSpace=true
    bOverrideAngularDamping=true
    bUseSphericalLimits=true
    simulationSpace=RootRelative
    alphaBoolBlend {blendOption=Linear}
    boundBone {boneName=dyn_hair_r_a_05}
    componentAppliedLinearAccClamp=<1500 1500 1500>
    componentLinearAccScale=<0.6 0.6 0.6>
    componentLinearVelScale=<0.6 0.6 0.6>
    physicsBodyDefinitions [{
      sphereCollisionRadius=0
      boundBone {boneName=dyn_hair_r_a_05}
      boxExtents=<12 15 15>
      constraintSetup {angularLimitsMax=<1 5 0>, angularLimitsMin=<-1 -5 -15>, angularTarget=<1 0 0>}
      localJointOffset=<1 0 0>
     }]
    relativeSpaceBone {boneName=root}
    sphericalLimits [
     {limitRadius=7, drivingBone {boneName=upperarm_r}, sphereLocalOffset=<-15 0 0>}
     {limitRadius=15, drivingBone {boneName=spine_04}, sphereLocalOffset=<0 5 0>}
    ]
   }
   n77 = Anim Dynamics - Bone: dyn_hair_r_b_02(ComponentPose=@n76, Alpha=0.4, ExternalForce="0,0,0") {
    angularDampingOverride=0.98
    angularSpringConstant=150
    bAngularSpring=true
    bGravityOverrideInSimSpace=true
    bOverrideAngularDamping=true
    simulationSpace=RootRelative
    alphaBoolBlend {blendOption=Linear}
    boundBone {boneName=dyn_hair_r_b_02}
    componentAppliedLinearAccClamp=<1500 1500 1500>
    componentLinearAccScale=<0.6 0.6 0.6>
    componentLinearVelScale=<0.6 0.6 0.6>
    physicsBodyDefinitions [{
      sphereCollisionRadius=0
      boundBone {boneName=dyn_hair_r_b_02}
      boxExtents=<12 15 15>
      constraintSetup {angularLimitsMax=<1 5 0>, angularLimitsMin=<-1 -5 -15>, angularTarget=<1 0 0>}
      localJointOffset=<1 0 0>
     }]
    relativeSpaceBone {boneName=root}
   }
   n78 = Anim Dynamics - Bone: dyn_hair_r_b_03(ComponentPose=@n77, Alpha=0.4, ExternalForce="0,0,0") {
    angularDampingOverride=0.98
    angularSpringConstant=150
    bAngularSpring=true
    bGravityOverrideInSimSpace=true
    bOverrideAngularDamping=true
    simulationSpace=RootRelative
    alphaBoolBlend {blendOption=Linear}
    boundBone {boneName=dyn_hair_r_b_03}
    componentAppliedLinearAccClamp=<1500 1500 1500>
    componentLinearAccScale=<0.6 0.6 0.6>
    componentLinearVelScale=<0.6 0.6 0.6>
    physicsBodyDefinitions [{
      sphereCollisionRadius=0
      boundBone {boneName=dyn_hair_r_b_03}
      boxExtents=<12 15 15>
      constraintSetup {angularLimitsMax=<1 5 0>, angularLimitsMin=<-1 -5 -15>, angularTarget=<1 0 0>}
      localJointOffset=<1 0 0>
     }]
    relativeSpaceBone {boneName=root}
   }
   n79 = Anim Dynamics - Bone: dyn_hair_r_b_04(ComponentPose=@n78, Alpha=0.4, ExternalForce="0,0,0") {
    angularDampingOverride=0.98
    angularSpringConstant=150
    bAngularSpring=true
    bGravityOverrideInSimSpace=true
    bOverrideAngularDamping=true
    bUseSphericalLimits=true
    simulationSpace=RootRelative
    alphaBoolBlend {blendOption=Linear}
    boundBone {boneName=dyn_hair_r_b_04}
    componentAppliedLinearAccClamp=<1500 1500 1500>
    componentLinearAccScale=<0.6 0.6 0.6>
    componentLinearVelScale=<0.6 0.6 0.6>
    physicsBodyDefinitions [{
      sphereCollisionRadius=0
      boundBone {boneName=dyn_hair_r_b_04}
      boxExtents=<12 15 15>
      constraintSetup {angularLimitsMax=<1 5 15>, angularLimitsMin=<-1 -5 0>, angularTarget=<1 0 0>}
      localJointOffset=<1 0 0>
     }]
    relativeSpaceBone {boneName=root}
    sphericalLimits [
     {limitRadius=7, drivingBone {boneName=upperarm_r}, sphereLocalOffset=<-15 0 0>}
     {limitRadius=15, drivingBone {boneName=spine_04}, sphereLocalOffset=<0 5 0>}
    ]
   }
   n80 = Anim Dynamics - Bone: dyn_hair_r_b_05(ComponentPose=@n79, Alpha=0.4, ExternalForce="0,0,0") {
    angularDampingOverride=0.98
    angularSpringConstant=150
    bAngularSpring=true
    bGravityOverrideInSimSpace=true
    bOverrideAngularDamping=true
    bUseSphericalLimits=true
    simulationSpace=RootRelative
    alphaBoolBlend {blendOption=Linear}
    boundBone {boneName=dyn_hair_r_b_05}
    componentAppliedLinearAccClamp=<1500 1500 1500>
    componentLinearAccScale=<0.6 0.6 0.6>
    componentLinearVelScale=<0.6 0.6 0.6>
    physicsBodyDefinitions [{
      sphereCollisionRadius=0
      boundBone {boneName=dyn_hair_r_b_05}
      boxExtents=<12 15 15>
      constraintSetup {angularLimitsMax=<1 5 15>, angularLimitsMin=<-1 -5 0>, angularTarget=<1 0 0>}
      localJointOffset=<1 0 0>
     }]
    relativeSpaceBone {boneName=root}
    sphericalLimits [
     {limitRadius=7, drivingBone {boneName=upperarm_r}, sphereLocalOffset=<-15 0 0>}
     {limitRadius=15, drivingBone {boneName=spine_04}, sphereLocalOffset=<0 5 0>}
    ]
   }
   n81 = Anim Dynamics - Bone: dyn_hair_r_c_02(ComponentPose=@n80, Alpha=0.2, ExternalForce="0,0,0") {
    angularDampingOverride=0.98
    angularSpringConstant=150
    bAngularSpring=true
    bGravityOverrideInSimSpace=true
    bOverrideAngularDamping=true
    simulationSpace=RootRelative
    alphaBoolBlend {blendOption=Linear}
    boundBone {boneName=dyn_hair_r_c_02}
    componentAppliedLinearAccClamp=<1500 1500 1500>
    componentLinearAccScale=<0.6 0.6 0.6>
    componentLinearVelScale=<0.6 0.6 0.6>
    physicsBodyDefinitions [{
      sphereCollisionRadius=0
      boundBone {boneName=dyn_hair_r_c_02}
      boxExtents=<12 15 15>
      constraintSetup {angularLimitsMax=<1 0 5>, angularLimitsMin=<-1 -10 -5>, angularTarget=<1 0 0>}
      localJointOffset=<1 0 0>
     }]
    relativeSpaceBone {boneName=root}
   }
   n82 = Anim Dynamics - Bone: dyn_hair_r_c_03(ComponentPose=@n81, Alpha=0.2, ExternalForce="0,0,0") {
    angularDampingOverride=0.98
    angularSpringConstant=150
    bAngularSpring=true
    bGravityOverrideInSimSpace=true
    bOverrideAngularDamping=true
    simulationSpace=RootRelative
    alphaBoolBlend {blendOption=Linear}
    boundBone {boneName=dyn_hair_r_c_03}
    componentAppliedLinearAccClamp=<1500 1500 1500>
    componentLinearAccScale=<0.6 0.6 0.6>
    componentLinearVelScale=<0.6 0.6 0.6>
    physicsBodyDefinitions [{
      sphereCollisionRadius=0
      boundBone {boneName=dyn_hair_r_c_03}
      boxExtents=<12 15 15>
      constraintSetup {angularLimitsMax=<1 0 5>, angularLimitsMin=<-1 -10 -5>, angularTarget=<1 0 0>}
      localJointOffset=<1 0 0>
     }]
    relativeSpaceBone {boneName=root}
   }
   n83 = Anim Dynamics - Bone: dyn_hair_r_c_04(ComponentPose=@n82, Alpha=0.2, ExternalForce="0,0,0") {
    angularDampingOverride=0.98
    angularSpringConstant=150
    bAngularSpring=true
    bGravityOverrideInSimSpace=true
    bOverrideAngularDamping=true
    bUseSphericalLimits=true
    simulationSpace=RootRelative
    alphaBoolBlend {blendOption=Linear}
    boundBone {boneName=dyn_hair_r_c_04}
    componentAppliedLinearAccClamp=<1500 1500 1500>
    componentLinearAccScale=<0.6 0.6 0.6>
    componentLinearVelScale=<0.6 0.6 0.6>
    physicsBodyDefinitions [{
      sphereCollisionRadius=0
      boundBone {boneName=dyn_hair_r_c_04}
      boxExtents=<12 15 15>
      constraintSetup {angularLimitsMax=<1 0 5>, angularLimitsMin=<-1 -10 -5>, angularTarget=<1 0 0>}
      localJointOffset=<1 0 0>
     }]
    relativeSpaceBone {boneName=root}
    sphericalLimits [
     {limitRadius=7, drivingBone {boneName=upperarm_r}, sphereLocalOffset=<-15 0 0>}
     {limitRadius=15, drivingBone {boneName=spine_04}, sphereLocalOffset=<0 5 0>}
    ]
   }
   n84 = Anim Dynamics - Bone: dyn_hair_r_c_05(ComponentPose=@n83, Alpha=0.2, ExternalForce="0,0,0") {
    angularDampingOverride=0.98
    angularSpringConstant=150
    bAngularSpring=true
    bGravityOverrideInSimSpace=true
    bOverrideAngularDamping=true
    bUseSphericalLimits=true
    simulationSpace=RootRelative
    alphaBoolBlend {blendOption=Linear}
    boundBone {boneName=dyn_hair_r_c_05}
    componentAppliedLinearAccClamp=<1500 1500 1500>
    componentLinearAccScale=<0.6 0.6 0.6>
    componentLinearVelScale=<0.6 0.6 0.6>
    physicsBodyDefinitions [{
      sphereCollisionRadius=0
      boundBone {boneName=dyn_hair_r_c_05}
      boxExtents=<12 15 15>
      constraintSetup {angularLimitsMax=<1 0 5>, angularLimitsMin=<-1 -10 -5>, angularTarget=<1 0 0>}
      localJointOffset=<1 0 0>
     }]
    relativeSpaceBone {boneName=root}
    sphericalLimits [
     {limitRadius=7, drivingBone {boneName=upperarm_r}, sphereLocalOffset=<-15 0 0>}
     {limitRadius=15, drivingBone {boneName=spine_04}, sphereLocalOffset=<0 5 0>}
    ]
   }
   n85 = Anim Dynamics - Bone: dyn_hair_r_d_02(ComponentPose=@n84, Alpha=0.2, ExternalForce="0,0,0") {
    angularDampingOverride=0.98
    angularSpringConstant=150
    bAngularSpring=true
    bGravityOverrideInSimSpace=true
    bOverrideAngularDamping=true
    simulationSpace=RootRelative
    alphaBoolBlend {blendOption=Linear}
    boundBone {boneName=dyn_hair_r_d_02}
    componentAppliedLinearAccClamp=<1500 1500 1500>
    componentLinearAccScale=<0.6 0.6 0.6>
    componentLinearVelScale=<0.6 0.6 0.6>
    physicsBodyDefinitions [{
      sphereCollisionRadius=0
      boundBone {boneName=dyn_hair_r_d_02}
      boxExtents=<12 15 15>
      constraintSetup {angularLimitsMax=<1 5 0>, angularLimitsMin=<-1 -5 -15>, angularTarget=<1 0 0>}
      localJointOffset=<1 0 0>
     }]
    relativeSpaceBone {boneName=root}
   }
   n86 = Anim Dynamics - Bone: dyn_hair_r_d_03(ComponentPose=@n85, Alpha=0.2, ExternalForce="0,0,0") {
    angularDampingOverride=0.98
    angularSpringConstant=150
    bAngularSpring=true
    bGravityOverrideInSimSpace=true
    bOverrideAngularDamping=true
    simulationSpace=RootRelative
    alphaBoolBlend {blendOption=Linear}
    boundBone {boneName=dyn_hair_r_d_03}
    componentAppliedLinearAccClamp=<1500 1500 1500>
    componentLinearAccScale=<0.6 0.6 0.6>
    componentLinearVelScale=<0.6 0.6 0.6>
    physicsBodyDefinitions [{
      sphereCollisionRadius=0
      boundBone {boneName=dyn_hair_r_d_03}
      boxExtents=<12 15 15>
      constraintSetup {angularLimitsMax=<1 5 0>, angularLimitsMin=<-1 -5 -15>, angularTarget=<1 0 0>}
      localJointOffset=<1 0 0>
     }]
    relativeSpaceBone {boneName=root}
   }
   n87 = Anim Dynamics - Bone: dyn_hair_r_d_04(ComponentPose=@n86, Alpha=0.2, ExternalForce="0,0,0") {
    angularDampingOverride=0.98
    angularSpringConstant=150
    bAngularSpring=true
    bGravityOverrideInSimSpace=true
    bOverrideAngularDamping=true
    bUseSphericalLimits=true
    simulationSpace=RootRelative
    alphaBoolBlend {blendOption=Linear}
    boundBone {boneName=dyn_hair_r_d_04}
    componentAppliedLinearAccClamp=<1500 1500 1500>
    componentLinearAccScale=<0.6 0.6 0.6>
    componentLinearVelScale=<0.6 0.6 0.6>
    physicsBodyDefinitions [{
      sphereCollisionRadius=0
      boundBone {boneName=dyn_hair_r_d_04}
      boxExtents=<12 15 15>
      constraintSetup {angularLimitsMax=<1 5 0>, angularLimitsMin=<-1 -5 -15>, angularTarget=<1 0 0>}
      localJointOffset=<1 0 0>
     }]
    relativeSpaceBone {boneName=root}
    sphericalLimits [
     {limitRadius=7, drivingBone {boneName=upperarm_r}, sphereLocalOffset=<-15 0 0>}
     {limitRadius=15, drivingBone {boneName=spine_04}, sphereLocalOffset=<0 5 0>}
    ]
   }
   n88 = Anim Dynamics - Bone: dyn_hair_r_d_05(ComponentPose=@n87, Alpha=0.2, ExternalForce="0,0,0") {
    angularDampingOverride=0.98
    angularSpringConstant=150
    bAngularSpring=true
    bGravityOverrideInSimSpace=true
    bOverrideAngularDamping=true
    bUseSphericalLimits=true
    simulationSpace=RootRelative
    alphaBoolBlend {blendOption=Linear}
    boundBone {boneName=dyn_hair_r_d_05}
    componentAppliedLinearAccClamp=<1500 1500 1500>
    componentLinearAccScale=<0.6 0.6 0.6>
    componentLinearVelScale=<0.6 0.6 0.6>
    physicsBodyDefinitions [{
      sphereCollisionRadius=0
      boundBone {boneName=dyn_hair_r_d_05}
      boxExtents=<12 15 15>
      constraintSetup {angularLimitsMax=<1 5 0>, angularLimitsMin=<-1 -5 -15>, angularTarget=<1 0 0>}
      localJointOffset=<1 0 0>
     }]
    relativeSpaceBone {boneName=root}
    sphericalLimits [
     {limitRadius=7, drivingBone {boneName=upperarm_r}, sphereLocalOffset=<-15 0 0>}
     {limitRadius=15, drivingBone {boneName=spine_04}, sphereLocalOffset=<0 5 0>}
    ]
   }
   n89 = Anim Dynamics - Bone: dyn_hair_r_e_02(ComponentPose=@n88, Alpha=0.2, ExternalForce="0,0,0") {
    angularDampingOverride=0.98
    angularSpringConstant=150
    bAngularSpring=true
    bGravityOverrideInSimSpace=true
    bOverrideAngularDamping=true
    simulationSpace=RootRelative
    alphaBoolBlend {blendOption=Linear}
    boundBone {boneName=dyn_hair_r_e_02}
    componentAppliedLinearAccClamp=<1500 1500 1500>
    componentLinearAccScale=<0.6 0.6 0.6>
    componentLinearVelScale=<0.6 0.6 0.6>
    physicsBodyDefinitions [{
      sphereCollisionRadius=0
      boundBone {boneName=dyn_hair_r_e_02}
      boxExtents=<12 15 15>
      constraintSetup {angularLimitsMax=<1 0 0>, angularLimitsMin=<-1 -10 -15>, angularTarget=<1 0 0>}
      localJointOffset=<1 0 0>
     }]
    relativeSpaceBone {boneName=root}
   }
   n90 = Anim Dynamics - Bone: dyn_hair_r_e_03(ComponentPose=@n89, Alpha=0.2, ExternalForce="0,0,0") {
    angularDampingOverride=0.98
    angularSpringConstant=150
    bAngularSpring=true
    bGravityOverrideInSimSpace=true
    bOverrideAngularDamping=true
    simulationSpace=RootRelative
    alphaBoolBlend {blendOption=Linear}
    boundBone {boneName=dyn_hair_r_e_03}
    componentAppliedLinearAccClamp=<1500 1500 1500>
    componentLinearAccScale=<0.6 0.6 0.6>
    componentLinearVelScale=<0.6 0.6 0.6>
    physicsBodyDefinitions [{
      sphereCollisionRadius=0
      boundBone {boneName=dyn_hair_r_e_03}
      boxExtents=<12 15 15>
      constraintSetup {angularLimitsMax=<1 0 0>, angularLimitsMin=<-1 -10 -15>, angularTarget=<1 0 0>}
      localJointOffset=<1 0 0>
     }]
    relativeSpaceBone {boneName=root}
   }
   n91 = Anim Dynamics - Bone: dyn_hair_r_e_04(ComponentPose=@n90, Alpha=0.2, ExternalForce="0,0,0") {
    angularDampingOverride=0.98
    angularSpringConstant=150
    bAngularSpring=true
    bGravityOverrideInSimSpace=true
    bOverrideAngularDamping=true
    bUseSphericalLimits=true
    simulationSpace=RootRelative
    alphaBoolBlend {blendOption=Linear}
    boundBone {boneName=dyn_hair_r_e_04}
    componentAppliedLinearAccClamp=<1500 1500 1500>
    componentLinearAccScale=<0.6 0.6 0.6>
    componentLinearVelScale=<0.6 0.6 0.6>
    physicsBodyDefinitions [{
      sphereCollisionRadius=0
      boundBone {boneName=dyn_hair_r_e_04}
      boxExtents=<12 15 15>
      constraintSetup {angularLimitsMax=<1 0 0>, angularLimitsMin=<-1 -10 -15>, angularTarget=<1 0 0>}
      localJointOffset=<1 0 0>
     }]
    relativeSpaceBone {boneName=root}
    sphericalLimits [
     {limitRadius=7, drivingBone {boneName=upperarm_r}, sphereLocalOffset=<-15 0 0>}
     {limitRadius=15, drivingBone {boneName=spine_04}, sphereLocalOffset=<0 5 0>}
    ]
   }
   n92 = Anim Dynamics - Bone: dyn_hair_r_e_05(ComponentPose=@n91, Alpha=0.2, ExternalForce="0,0,0") {
    angularDampingOverride=0.98
    angularSpringConstant=150
    bAngularSpring=true
    bGravityOverrideInSimSpace=true
    bOverrideAngularDamping=true
    bUseSphericalLimits=true
    simulationSpace=RootRelative
    alphaBoolBlend {blendOption=Linear}
    boundBone {boneName=dyn_hair_r_e_05}
    componentAppliedLinearAccClamp=<1500 1500 1500>
    componentLinearAccScale=<0.6 0.6 0.6>
    componentLinearVelScale=<0.6 0.6 0.6>
    physicsBodyDefinitions [{
      sphereCollisionRadius=0
      boundBone {boneName=dyn_hair_r_e_05}
      boxExtents=<12 15 15>
      constraintSetup {angularLimitsMax=<1 0 0>, angularLimitsMin=<-1 -10 -15>, angularTarget=<1 0 0>}
      localJointOffset=<1 0 0>
     }]
    relativeSpaceBone {boneName=root}
    sphericalLimits [
     {limitRadius=7, drivingBone {boneName=upperarm_r}, sphereLocalOffset=<-15 0 0>}
     {limitRadius=15, drivingBone {boneName=spine_04}, sphereLocalOffset=<0 5 0>}
    ]
   }
   n93 = Anim Dynamics - Bone: dyn_hair_r_f_02(ComponentPose=@n92, Alpha=0.4, ExternalForce="0,0,0") {
    angularDampingOverride=0.98
    angularSpringConstant=150
    bAngularSpring=true
    bGravityOverrideInSimSpace=true
    bOverrideAngularDamping=true
    simulationSpace=RootRelative
    alphaBoolBlend {blendOption=Linear}
    boundBone {boneName=dyn_hair_r_f_02}
    componentAppliedLinearAccClamp=<1500 1500 1500>
    componentLinearAccScale=<0.6 0.6 0.6>
    componentLinearVelScale=<0.6 0.6 0.6>
    physicsBodyDefinitions [{
      sphereCollisionRadius=0
      boundBone {boneName=dyn_hair_r_f_02}
      boxExtents=<12 15 15>
      constraintSetup {angularLimitsMax=<1 8 0>, angularLimitsMin=<-1 -2 -15>, angularTarget=<1 0 0>}
      localJointOffset=<1 0 0>
     }]
    relativeSpaceBone {boneName=root}
   }
   n94 = Anim Dynamics - Bone: dyn_hair_r_f_03(ComponentPose=@n93, Alpha=0.4, ExternalForce="0,0,0") {
    angularDampingOverride=0.98
    angularSpringConstant=150
    bAngularSpring=true
    bGravityOverrideInSimSpace=true
    bOverrideAngularDamping=true
    simulationSpace=RootRelative
    alphaBoolBlend {blendOption=Linear}
    boundBone {boneName=dyn_hair_r_f_03}
    componentAppliedLinearAccClamp=<1500 1500 1500>
    componentLinearAccScale=<0.6 0.6 0.6>
    componentLinearVelScale=<0.6 0.6 0.6>
    physicsBodyDefinitions [{
      sphereCollisionRadius=0
      boundBone {boneName=dyn_hair_r_f_03}
      boxExtents=<12 15 15>
      constraintSetup {angularLimitsMax=<1 8 0>, angularLimitsMin=<-1 -2 -15>, angularTarget=<1 0 0>}
      localJointOffset=<1 0 0>
     }]
    relativeSpaceBone {boneName=root}
   }
   n95 = Anim Dynamics - Bone: dyn_hair_r_f_04(ComponentPose=@n94, Alpha=0.4, ExternalForce="0,0,0") {
    angularDampingOverride=0.98
    angularSpringConstant=150
    bAngularSpring=true
    bGravityOverrideInSimSpace=true
    bOverrideAngularDamping=true
    bUseSphericalLimits=true
    simulationSpace=RootRelative
    alphaBoolBlend {blendOption=Linear}
    boundBone {boneName=dyn_hair_r_f_04}
    componentAppliedLinearAccClamp=<1500 1500 1500>
    componentLinearAccScale=<0.6 0.6 0.6>
    componentLinearVelScale=<0.6 0.6 0.6>
    physicsBodyDefinitions [{
      sphereCollisionRadius=0
      boundBone {boneName=dyn_hair_r_f_04}
      boxExtents=<12 15 15>
      constraintSetup {angularLimitsMax=<1 8 0>, angularLimitsMin=<-1 -2 -15>, angularTarget=<1 0 0>}
      localJointOffset=<1 0 0>
     }]
    relativeSpaceBone {boneName=root}
    sphericalLimits [
     {limitRadius=7, drivingBone {boneName=upperarm_r}, sphereLocalOffset=<-15 0 0>}
     {limitRadius=15, drivingBone {boneName=spine_04}, sphereLocalOffset=<0 5 0>}
    ]
   }
   n96 = Anim Dynamics - Bone: dyn_hair_r_f_05(ComponentPose=@n95, Alpha=0.4, ExternalForce="0,0,0") {
    angularDampingOverride=0.98
    angularSpringConstant=150
    bAngularSpring=true
    bGravityOverrideInSimSpace=true
    bOverrideAngularDamping=true
    bUseSphericalLimits=true
    simulationSpace=RootRelative
    alphaBoolBlend {blendOption=Linear}
    boundBone {boneName=dyn_hair_r_f_05}
    componentAppliedLinearAccClamp=<1500 1500 1500>
    componentLinearAccScale=<0.6 0.6 0.6>
    componentLinearVelScale=<0.6 0.6 0.6>
    physicsBodyDefinitions [{
      sphereCollisionRadius=0
      boundBone {boneName=dyn_hair_r_f_05}
      boxExtents=<12 15 15>
      constraintSetup {angularLimitsMax=<1 8 0>, angularLimitsMin=<-1 -2 -15>, angularTarget=<1 0 0>}
      localJointOffset=<1 0 0>
     }]
    relativeSpaceBone {boneName=root}
    sphericalLimits [
     {limitRadius=7, drivingBone {boneName=upperarm_r}, sphereLocalOffset=<-15 0 0>}
     {limitRadius=15, drivingBone {boneName=spine_04}, sphereLocalOffset=<0 5 0>}
    ]
   }
   n97 = Anim Dynamics - Bone: dyn_hair_l_a_02(ComponentPose=@n96, Alpha=0.5, ExternalForce="0,0,0") {
    angularDampingOverride=0.98
    angularSpringConstant=150
    bAngularSpring=true
    bGravityOverrideInSimSpace=true
    bOverrideAngularDamping=true
    simulationSpace=RootRelative
    alphaBoolBlend {blendOption=Linear}
    boundBone {boneName=dyn_hair_l_a_02}
    componentAppliedLinearAccClamp=<1500 1500 1500>
    componentLinearAccScale=<0.6 0.6 0.6>
    componentLinearVelScale=<0.6 0.6 0.6>
    physicsBodyDefinitions [{
      sphereCollisionRadius=0
      boundBone {boneName=dyn_hair_l_a_02}
      boxExtents=<12 15 15>
      constraintSetup {angularLimitsMax=<1 5 0>, angularLimitsMin=<-1 -5 -15>, angularTarget=<1 0 0>}
      localJointOffset=<1 0 0>
     }]
    relativeSpaceBone {boneName=root}
   }
   n98 = Anim Dynamics - Bone: dyn_hair_l_a_03(ComponentPose=@n97, Alpha=0.5, ExternalForce="0,0,0") {
    angularDampingOverride=0.98
    angularSpringConstant=150
    bAngularSpring=true
    bGravityOverrideInSimSpace=true
    bOverrideAngularDamping=true
    simulationSpace=RootRelative
    alphaBoolBlend {blendOption=Linear}
    boundBone {boneName=dyn_hair_l_a_03}
    componentAppliedLinearAccClamp=<1500 1500 1500>
    componentLinearAccScale=<0.6 0.6 0.6>
    componentLinearVelScale=<0.6 0.6 0.6>
    physicsBodyDefinitions [{
      sphereCollisionRadius=0
      boundBone {boneName=dyn_hair_l_a_03}
      boxExtents=<12 15 15>
      constraintSetup {angularLimitsMax=<1 5 0>, angularLimitsMin=<-1 -5 -15>, angularTarget=<1 0 0>}
      localJointOffset=<1 0 0>
     }]
    relativeSpaceBone {boneName=root}
   }
   n99 = Anim Dynamics - Bone: dyn_hair_l_a_04(ComponentPose=@n98, Alpha=0.5, ExternalForce="0,0,0") {
    angularDampingOverride=0.98
    angularSpringConstant=150
    bAngularSpring=true
    bGravityOverrideInSimSpace=true
    bOverrideAngularDamping=true
    bUseSphericalLimits=true
    simulationSpace=RootRelative
    alphaBoolBlend {blendOption=Linear}
    boundBone {boneName=dyn_hair_l_a_04}
    componentAppliedLinearAccClamp=<1500 1500 1500>
    componentLinearAccScale=<0.6 0.6 0.6>
    componentLinearVelScale=<0.6 0.6 0.6>
    physicsBodyDefinitions [{
      sphereCollisionRadius=0
      boundBone {boneName=dyn_hair_l_a_04}
      boxExtents=<12 15 15>
      constraintSetup {angularLimitsMax=<1 5 0>, angularLimitsMin=<-1 -5 -15>, angularTarget=<1 0 0>}
      localJointOffset=<1 0 0>
     }]
    relativeSpaceBone {boneName=root}
    sphericalLimits [
     {limitRadius=7, drivingBone {boneName=upperarm_l}, sphereLocalOffset=<15 0 0>}
     {limitRadius=15, drivingBone {boneName=spine_04}, sphereLocalOffset=<0 5 0>}
    ]
   }
   n100 = Anim Dynamics - Bone: dyn_hair_l_a_05(ComponentPose=@n99, Alpha=0.5, ExternalForce="0,0,0") {
    angularDampingOverride=0.98
    angularSpringConstant=150
    bAngularSpring=true
    bGravityOverrideInSimSpace=true
    bOverrideAngularDamping=true
    bUseSphericalLimits=true
    simulationSpace=RootRelative
    alphaBoolBlend {blendOption=Linear}
    boundBone {boneName=dyn_hair_l_a_05}
    componentAppliedLinearAccClamp=<1500 1500 1500>
    componentLinearAccScale=<0.6 0.6 0.6>
    componentLinearVelScale=<0.6 0.6 0.6>
    physicsBodyDefinitions [{
      sphereCollisionRadius=0
      boundBone {boneName=dyn_hair_l_a_05}
      boxExtents=<12 15 15>
      constraintSetup {angularLimitsMax=<1 5 0>, angularLimitsMin=<-1 -5 -15>, angularTarget=<1 0 0>}
      localJointOffset=<1 0 0>
     }]
    relativeSpaceBone {boneName=root}
    sphericalLimits [
     {limitRadius=7, drivingBone {boneName=upperarm_l}, sphereLocalOffset=<15 0 0>}
     {limitRadius=15, drivingBone {boneName=spine_04}, sphereLocalOffset=<0 5 0>}
    ]
   }
   n101 = Anim Dynamics - Bone: dyn_hair_l_b_02(ComponentPose=@n100, Alpha=0.4, ExternalForce="0,0,0") {
    angularDampingOverride=0.98
    angularSpringConstant=150
    bAngularSpring=true
    bGravityOverrideInSimSpace=true
    bOverrideAngularDamping=true
    simulationSpace=RootRelative
    alphaBoolBlend {blendOption=Linear}
    boundBone {boneName=dyn_hair_l_b_02}
    componentAppliedLinearAccClamp=<1500 1500 1500>
    componentLinearAccScale=<0.6 0.6 0.6>
    componentLinearVelScale=<0.6 0.6 0.6>
    physicsBodyDefinitions [{
      sphereCollisionRadius=0
      boundBone {boneName=dyn_hair_l_b_02}
      boxExtents=<12 15 15>
      constraintSetup {angularLimitsMax=<1 8 0>, angularLimitsMin=<-1 -2 -15>, angularTarget=<1 0 0>}
      localJointOffset=<1 0 0>
     }]
    relativeSpaceBone {boneName=root}
   }
   n102 = Anim Dynamics - Bone: dyn_hair_l_b_03(ComponentPose=@n101, Alpha=0.4, ExternalForce="0,0,0") {
    angularDampingOverride=0.98
    angularSpringConstant=150
    bAngularSpring=true
    bGravityOverrideInSimSpace=true
    bOverrideAngularDamping=true
    simulationSpace=RootRelative
    alphaBoolBlend {blendOption=Linear}
    boundBone {boneName=dyn_hair_l_b_03}
    componentAppliedLinearAccClamp=<1500 1500 1500>
    componentLinearAccScale=<0.6 0.6 0.6>
    componentLinearVelScale=<0.6 0.6 0.6>
    physicsBodyDefinitions [{
      sphereCollisionRadius=0
      boundBone {boneName=dyn_hair_l_b_03}
      boxExtents=<12 15 15>
      constraintSetup {angularLimitsMax=<1 8 0>, angularLimitsMin=<-1 -2 -15>, angularTarget=<1 0 0>}
      localJointOffset=<1 0 0>
     }]
    relativeSpaceBone {boneName=root}
   }
   n103 = Anim Dynamics - Bone: dyn_hair_l_b_04(ComponentPose=@n102, Alpha=0.4, ExternalForce="0,0,0") {
    angularDampingOverride=0.98
    angularSpringConstant=150
    bAngularSpring=true
    bGravityOverrideInSimSpace=true
    bOverrideAngularDamping=true
    bUseSphericalLimits=true
    simulationSpace=RootRelative
    alphaBoolBlend {blendOption=Linear}
    boundBone {boneName=dyn_hair_l_b_04}
    componentAppliedLinearAccClamp=<1500 1500 1500>
    componentLinearAccScale=<0.6 0.6 0.6>
    componentLinearVelScale=<0.6 0.6 0.6>
    physicsBodyDefinitions [{
      sphereCollisionRadius=0
      boundBone {boneName=dyn_hair_l_b_04}
      boxExtents=<12 15 15>
      constraintSetup {angularLimitsMax=<1 8 0>, angularLimitsMin=<-1 -2 -15>, angularTarget=<1 0 0>}
      localJointOffset=<1 0 0>
     }]
    relativeSpaceBone {boneName=root}
    sphericalLimits [
     {limitRadius=7, drivingBone {boneName=upperarm_l}, sphereLocalOffset=<15 0 0>}
     {limitRadius=15, drivingBone {boneName=spine_04}, sphereLocalOffset=<0 5 0>}
    ]
   }
   n104 = Anim Dynamics - Bone: dyn_hair_l_b_05(ComponentPose=@n103, Alpha=0.4, ExternalForce="0,0,0") {
    angularDampingOverride=0.98
    angularSpringConstant=150
    bAngularSpring=true
    bGravityOverrideInSimSpace=true
    bOverrideAngularDamping=true
    bUseSphericalLimits=true
    simulationSpace=RootRelative
    alphaBoolBlend {blendOption=Linear}
    boundBone {boneName=dyn_hair_l_b_05}
    componentAppliedLinearAccClamp=<1500 1500 1500>
    componentLinearAccScale=<0.6 0.6 0.6>
    componentLinearVelScale=<0.6 0.6 0.6>
    physicsBodyDefinitions [{
      sphereCollisionRadius=0
      boundBone {boneName=dyn_hair_l_b_05}
      boxExtents=<12 15 15>
      constraintSetup {angularLimitsMax=<1 8 0>, angularLimitsMin=<-1 -2 -15>, angularTarget=<1 0 0>}
      localJointOffset=<1 0 0>
     }]
    relativeSpaceBone {boneName=root}
    sphericalLimits [
     {limitRadius=7, drivingBone {boneName=upperarm_l}, sphereLocalOffset=<15 0 0>}
     {limitRadius=15, drivingBone {boneName=spine_04}, sphereLocalOffset=<0 5 0>}
    ]
   }
   n105 = Anim Dynamics - Bone: dyn_hair_l_c_02(ComponentPose=@n104, Alpha=0.2, ExternalForce="0,0,0") {
    angularDampingOverride=0.98
    angularSpringConstant=150
    bAngularSpring=true
    bGravityOverrideInSimSpace=true
    bOverrideAngularDamping=true
    simulationSpace=RootRelative
    alphaBoolBlend {blendOption=Linear}
    boundBone {boneName=dyn_hair_l_c_02}
    componentAppliedLinearAccClamp=<1500 1500 1500>
    componentLinearAccScale=<0.6 0.6 0.6>
    componentLinearVelScale=<0.6 0.6 0.6>
    physicsBodyDefinitions [{
      sphereCollisionRadius=0
      boundBone {boneName=dyn_hair_l_c_02}
      boxExtents=<12 15 15>
      constraintSetup {angularLimitsMax=<1 8 10>, angularLimitsMin=<-1 -2 0>, angularTarget=<1 0 0>}
      localJointOffset=<1 0 0>
     }]
    relativeSpaceBone {boneName=root}
   }
   n106 = Anim Dynamics - Bone: dyn_hair_l_c_03(ComponentPose=@n105, Alpha=0.2, ExternalForce="0,0,0") {
    angularDampingOverride=0.98
    angularSpringConstant=150
    bAngularSpring=true
    bGravityOverrideInSimSpace=true
    bOverrideAngularDamping=true
    simulationSpace=RootRelative
    alphaBoolBlend {blendOption=Linear}
    boundBone {boneName=dyn_hair_l_c_03}
    componentAppliedLinearAccClamp=<1500 1500 1500>
    componentLinearAccScale=<0.6 0.6 0.6>
    componentLinearVelScale=<0.6 0.6 0.6>
    physicsBodyDefinitions [{
      sphereCollisionRadius=0
      boundBone {boneName=dyn_hair_l_c_03}
      boxExtents=<12 15 15>
      constraintSetup {angularLimitsMax=<1 8 10>, angularLimitsMin=<-1 -2 0>, angularTarget=<1 0 0>}
      localJointOffset=<1 0 0>
     }]
    relativeSpaceBone {boneName=root}
   }
   n107 = Anim Dynamics - Bone: dyn_hair_l_c_04(ComponentPose=@n106, Alpha=0.2, ExternalForce="0,0,0") {
    angularDampingOverride=0.98
    angularSpringConstant=150
    bAngularSpring=true
    bGravityOverrideInSimSpace=true
    bOverrideAngularDamping=true
    bUseSphericalLimits=true
    simulationSpace=RootRelative
    alphaBoolBlend {blendOption=Linear}
    boundBone {boneName=dyn_hair_l_c_04}
    componentAppliedLinearAccClamp=<1500 1500 1500>
    componentLinearAccScale=<0.6 0.6 0.6>
    componentLinearVelScale=<0.6 0.6 0.6>
    physicsBodyDefinitions [{
      sphereCollisionRadius=0
      boundBone {boneName=dyn_hair_l_c_04}
      boxExtents=<12 15 15>
      constraintSetup {angularLimitsMax=<1 8 10>, angularLimitsMin=<-1 -2 0>, angularTarget=<1 0 0>}
      localJointOffset=<1 0 0>
     }]
    relativeSpaceBone {boneName=root}
    sphericalLimits [
     {limitRadius=7, drivingBone {boneName=upperarm_l}, sphereLocalOffset=<15 0 0>}
     {limitRadius=15, drivingBone {boneName=spine_04}, sphereLocalOffset=<0 5 0>}
    ]
   }
   n108 = Anim Dynamics - Bone: dyn_hair_l_c_05(ComponentPose=@n107, Alpha=0.2, ExternalForce="0,0,0") {
    angularDampingOverride=0.98
    angularSpringConstant=150
    bAngularSpring=true
    bGravityOverrideInSimSpace=true
    bOverrideAngularDamping=true
    bUseSphericalLimits=true
    simulationSpace=RootRelative
    alphaBoolBlend {blendOption=Linear}
    boundBone {boneName=dyn_hair_l_c_05}
    componentAppliedLinearAccClamp=<1500 1500 1500>
    componentLinearAccScale=<0.6 0.6 0.6>
    componentLinearVelScale=<0.6 0.6 0.6>
    physicsBodyDefinitions [{
      sphereCollisionRadius=0
      boundBone {boneName=dyn_hair_l_c_05}
      boxExtents=<12 15 15>
      constraintSetup {angularLimitsMax=<1 8 10>, angularLimitsMin=<-1 -2 0>, angularTarget=<1 0 0>}
      localJointOffset=<1 0 0>
     }]
    relativeSpaceBone {boneName=root}
    sphericalLimits [
     {limitRadius=7, drivingBone {boneName=upperarm_l}, sphereLocalOffset=<15 0 0>}
     {limitRadius=15, drivingBone {boneName=spine_04}, sphereLocalOffset=<0 5 0>}
    ]
   }
   n109 = Anim Dynamics - Bone: dyn_hair_l_d_02(ComponentPose=@n108, Alpha=0.2, ExternalForce="0,0,0") {
    angularDampingOverride=0.98
    angularSpringConstant=150
    bAngularSpring=true
    bGravityOverrideInSimSpace=true
    bOverrideAngularDamping=true
    simulationSpace=RootRelative
    alphaBoolBlend {blendOption=Linear}
    boundBone {boneName=dyn_hair_l_d_02}
    componentAppliedLinearAccClamp=<1500 1500 1500>
    componentLinearAccScale=<0.6 0.6 0.6>
    componentLinearVelScale=<0.6 0.6 0.6>
    physicsBodyDefinitions [{
      sphereCollisionRadius=0
      boundBone {boneName=dyn_hair_l_d_02}
      boxExtents=<12 15 15>
      constraintSetup {angularLimitsMax=<1 8 0>, angularLimitsMin=<-1 -2 -15>, angularTarget=<1 0 0>}
      localJointOffset=<1 0 0>
     }]
    relativeSpaceBone {boneName=root}
   }
   n110 = Anim Dynamics - Bone: dyn_hair_l_d_03(ComponentPose=@n109, Alpha=0.2, ExternalForce="0,0,0") {
    angularDampingOverride=0.98
    angularSpringConstant=150
    bAngularSpring=true
    bGravityOverrideInSimSpace=true
    bOverrideAngularDamping=true
    simulationSpace=RootRelative
    alphaBoolBlend {blendOption=Linear}
    boundBone {boneName=dyn_hair_l_d_03}
    componentAppliedLinearAccClamp=<1500 1500 1500>
    componentLinearAccScale=<0.6 0.6 0.6>
    componentLinearVelScale=<0.6 0.6 0.6>
    physicsBodyDefinitions [{
      sphereCollisionRadius=0
      boundBone {boneName=dyn_hair_l_d_03}
      boxExtents=<12 15 15>
      constraintSetup {angularLimitsMax=<1 8 0>, angularLimitsMin=<-1 -2 -15>, angularTarget=<1 0 0>}
      localJointOffset=<1 0 0>
     }]
    relativeSpaceBone {boneName=root}
   }
   n111 = Anim Dynamics - Bone: dyn_hair_l_d_04(ComponentPose=@n110, Alpha=0.2, ExternalForce="0,0,0") {
    angularDampingOverride=0.98
    angularSpringConstant=150
    bAngularSpring=true
    bGravityOverrideInSimSpace=true
    bOverrideAngularDamping=true
    bUseSphericalLimits=true
    simulationSpace=RootRelative
    alphaBoolBlend {blendOption=Linear}
    boundBone {boneName=dyn_hair_l_d_04}
    componentAppliedLinearAccClamp=<1500 1500 1500>
    componentLinearAccScale=<0.6 0.6 0.6>
    componentLinearVelScale=<0.6 0.6 0.6>
    physicsBodyDefinitions [{
      sphereCollisionRadius=0
      boundBone {boneName=dyn_hair_l_d_04}
      boxExtents=<12 15 15>
      constraintSetup {angularLimitsMax=<1 8 0>, angularLimitsMin=<-1 -2 -15>, angularTarget=<1 0 0>}
      localJointOffset=<1 0 0>
     }]
    relativeSpaceBone {boneName=root}
    sphericalLimits [
     {limitRadius=7, drivingBone {boneName=upperarm_l}, sphereLocalOffset=<15 0 0>}
     {limitRadius=15, drivingBone {boneName=spine_04}, sphereLocalOffset=<0 5 0>}
    ]
   }
   n112 = Anim Dynamics - Bone: dyn_hair_l_d_05(ComponentPose=@n111, Alpha=0.2, ExternalForce="0,0,0") {
    angularDampingOverride=0.98
    angularSpringConstant=150
    bAngularSpring=true
    bGravityOverrideInSimSpace=true
    bOverrideAngularDamping=true
    bUseSphericalLimits=true
    simulationSpace=RootRelative
    alphaBoolBlend {blendOption=Linear}
    boundBone {boneName=dyn_hair_l_d_05}
    componentAppliedLinearAccClamp=<1500 1500 1500>
    componentLinearAccScale=<0.6 0.6 0.6>
    componentLinearVelScale=<0.6 0.6 0.6>
    physicsBodyDefinitions [{
      sphereCollisionRadius=0
      boundBone {boneName=dyn_hair_l_d_05}
      boxExtents=<12 15 15>
      constraintSetup {angularLimitsMax=<1 8 0>, angularLimitsMin=<-1 -2 -15>, angularTarget=<1 0 0>}
      localJointOffset=<1 0 0>
     }]
    relativeSpaceBone {boneName=root}
    sphericalLimits [
     {limitRadius=7, drivingBone {boneName=upperarm_l}, sphereLocalOffset=<15 0 0>}
     {limitRadius=15, drivingBone {boneName=spine_04}, sphereLocalOffset=<0 5 0>}
    ]
   }
   n113 = Anim Dynamics - Bone: dyn_hair_l_e_02(ComponentPose=@n112, Alpha=0.2, ExternalForce="0,0,0") {
    angularDampingOverride=0.98
    angularSpringConstant=150
    bAngularSpring=true
    bGravityOverrideInSimSpace=true
    bOverrideAngularDamping=true
    simulationSpace=RootRelative
    alphaBoolBlend {blendOption=Linear}
    boundBone {boneName=dyn_hair_l_e_02}
    componentAppliedLinearAccClamp=<1500 1500 1500>
    componentLinearAccScale=<0.6 0.6 0.6>
    componentLinearVelScale=<0.6 0.6 0.6>
    physicsBodyDefinitions [{
      sphereCollisionRadius=0
      boundBone {boneName=dyn_hair_l_e_02}
      boxExtents=<12 15 15>
      constraintSetup {angularLimitsMax=<1 8 5>, angularLimitsMin=<-1 -2 -5>, angularTarget=<1 0 0>}
      localJointOffset=<1 0 0>
     }]
    relativeSpaceBone {boneName=root}
   }
   n114 = Anim Dynamics - Bone: dyn_hair_l_e_03(ComponentPose=@n113, Alpha=0.2, ExternalForce="0,0,0") {
    angularDampingOverride=0.98
    angularSpringConstant=150
    bAngularSpring=true
    bGravityOverrideInSimSpace=true
    bOverrideAngularDamping=true
    simulationSpace=RootRelative
    alphaBoolBlend {blendOption=Linear}
    boundBone {boneName=dyn_hair_l_e_03}
    componentAppliedLinearAccClamp=<1500 1500 1500>
    componentLinearAccScale=<0.6 0.6 0.6>
    componentLinearVelScale=<0.6 0.6 0.6>
    physicsBodyDefinitions [{
      sphereCollisionRadius=0
      boundBone {boneName=dyn_hair_l_e_03}
      boxExtents=<12 15 15>
      constraintSetup {angularLimitsMax=<1 8 5>, angularLimitsMin=<-1 -2 -5>, angularTarget=<1 0 0>}
      localJointOffset=<1 0 0>
     }]
    relativeSpaceBone {boneName=root}
   }
   n115 = Anim Dynamics - Bone: dyn_hair_l_e_04(ComponentPose=@n114, Alpha=0.2, ExternalForce="0,0,0") {
    angularDampingOverride=0.98
    angularSpringConstant=150
    bAngularSpring=true
    bGravityOverrideInSimSpace=true
    bOverrideAngularDamping=true
    bUseSphericalLimits=true
    simulationSpace=RootRelative
    alphaBoolBlend {blendOption=Linear}
    boundBone {boneName=dyn_hair_l_e_04}
    componentAppliedLinearAccClamp=<1500 1500 1500>
    componentLinearAccScale=<0.6 0.6 0.6>
    componentLinearVelScale=<0.6 0.6 0.6>
    physicsBodyDefinitions [{
      sphereCollisionRadius=0
      boundBone {boneName=dyn_hair_l_e_04}
      boxExtents=<12 15 15>
      constraintSetup {angularLimitsMax=<1 8 5>, angularLimitsMin=<-1 -2 -5>, angularTarget=<1 0 0>}
      localJointOffset=<1 0 0>
     }]
    relativeSpaceBone {boneName=root}
    sphericalLimits [
     {limitRadius=7, drivingBone {boneName=upperarm_l}, sphereLocalOffset=<15 0 0>}
     {limitRadius=15, drivingBone {boneName=spine_04}, sphereLocalOffset=<0 5 0>}
    ]
   }
   n116 = Anim Dynamics - Bone: dyn_hair_l_e_05(ComponentPose=@n115, Alpha=0.2, ExternalForce="0,0,0") {
    angularDampingOverride=0.98
    angularSpringConstant=150
    bAngularSpring=true
    bGravityOverrideInSimSpace=true
    bOverrideAngularDamping=true
    bUseSphericalLimits=true
    simulationSpace=RootRelative
    alphaBoolBlend {blendOption=Linear}
    boundBone {boneName=dyn_hair_l_e_05}
    componentAppliedLinearAccClamp=<1500 1500 1500>
    componentLinearAccScale=<0.6 0.6 0.6>
    componentLinearVelScale=<0.6 0.6 0.6>
    physicsBodyDefinitions [{
      sphereCollisionRadius=0
      boundBone {boneName=dyn_hair_l_e_05}
      boxExtents=<12 15 15>
      constraintSetup {angularLimitsMax=<1 8 5>, angularLimitsMin=<-1 -2 -5>, angularTarget=<1 0 0>}
      localJointOffset=<1 0 0>
     }]
    relativeSpaceBone {boneName=root}
    sphericalLimits [
     {limitRadius=7, drivingBone {boneName=upperarm_l}, sphereLocalOffset=<15 0 0>}
     {limitRadius=15, drivingBone {boneName=spine_04}, sphereLocalOffset=<0 5 0>}
    ]
   }
   n117 = Anim Dynamics - Bone: dyn_hair_l_f_02(ComponentPose=@n116, Alpha=0.4, ExternalForce="0,0,0") {
    angularDampingOverride=0.98
    angularSpringConstant=150
    bAngularSpring=true
    bGravityOverrideInSimSpace=true
    bOverrideAngularDamping=true
    simulationSpace=RootRelative
    alphaBoolBlend {blendOption=Linear}
    boundBone {boneName=dyn_hair_l_f_02}
    componentAppliedLinearAccClamp=<1500 1500 1500>
    componentLinearAccScale=<0.6 0.6 0.6>
    componentLinearVelScale=<0.6 0.6 0.6>
    physicsBodyDefinitions [{
      sphereCollisionRadius=0
      boundBone {boneName=dyn_hair_l_f_02}
      boxExtents=<12 15 15>
      constraintSetup {angularLimitsMax=<1 5 0>, angularLimitsMin=<-1 -5 -15>, angularTarget=<1 0 0>}
      localJointOffset=<1 0 0>
     }]
    relativeSpaceBone {boneName=root}
   }
   n118 = Anim Dynamics - Bone: dyn_hair_l_f_03(ComponentPose=@n117, Alpha=0.4, ExternalForce="0,0,0") {
    angularDampingOverride=0.98
    angularSpringConstant=150
    bAngularSpring=true
    bGravityOverrideInSimSpace=true
    bOverrideAngularDamping=true
    simulationSpace=RootRelative
    alphaBoolBlend {blendOption=Linear}
    boundBone {boneName=dyn_hair_l_f_03}
    componentAppliedLinearAccClamp=<1500 1500 1500>
    componentLinearAccScale=<0.6 0.6 0.6>
    componentLinearVelScale=<0.6 0.6 0.6>
    physicsBodyDefinitions [{
      sphereCollisionRadius=0
      boundBone {boneName=dyn_hair_l_f_03}
      boxExtents=<12 15 15>
      constraintSetup {angularLimitsMax=<1 5 0>, angularLimitsMin=<-1 -5 -15>, angularTarget=<1 0 0>}
      localJointOffset=<1 0 0>
     }]
    relativeSpaceBone {boneName=root}
   }
   n119 = Anim Dynamics - Bone: dyn_hair_l_f_04(ComponentPose=@n118, Alpha=0.4, ExternalForce="0,0,0") {
    angularDampingOverride=0.98
    angularSpringConstant=150
    bAngularSpring=true
    bGravityOverrideInSimSpace=true
    bOverrideAngularDamping=true
    bUseSphericalLimits=true
    simulationSpace=RootRelative
    alphaBoolBlend {blendOption=Linear}
    boundBone {boneName=dyn_hair_l_f_04}
    componentAppliedLinearAccClamp=<1500 1500 1500>
    componentLinearAccScale=<0.6 0.6 0.6>
    componentLinearVelScale=<0.6 0.6 0.6>
    physicsBodyDefinitions [{
      sphereCollisionRadius=0
      boundBone {boneName=dyn_hair_l_f_04}
      boxExtents=<12 15 15>
      constraintSetup {angularLimitsMax=<1 5 0>, angularLimitsMin=<-1 -5 -15>, angularTarget=<1 0 0>}
      localJointOffset=<1 0 0>
     }]
    relativeSpaceBone {boneName=root}
    sphericalLimits [
     {limitRadius=7, drivingBone {boneName=upperarm_l}, sphereLocalOffset=<15 0 0>}
     {limitRadius=15, drivingBone {boneName=spine_04}, sphereLocalOffset=<0 5 0>}
    ]
   }
   n120 = Anim Dynamics - Bone: dyn_hair_l_f_05(ComponentPose=@n119, Alpha=0.4, ExternalForce="0,0,0") {
    angularDampingOverride=0.98
    angularSpringConstant=150
    bAngularSpring=true
    bGravityOverrideInSimSpace=true
    bOverrideAngularDamping=true
    bUseSphericalLimits=true
    simulationSpace=RootRelative
    alphaBoolBlend {blendOption=Linear}
    boundBone {boneName=dyn_hair_l_f_05}
    componentAppliedLinearAccClamp=<1500 1500 1500>
    componentLinearAccScale=<0.6 0.6 0.6>
    componentLinearVelScale=<0.6 0.6 0.6>
    physicsBodyDefinitions [{
      sphereCollisionRadius=0
      boundBone {boneName=dyn_hair_l_f_05}
      boxExtents=<12 15 15>
      constraintSetup {angularLimitsMax=<1 5 0>, angularLimitsMin=<-1 -5 -15>, angularTarget=<1 0 0>}
      localJointOffset=<1 0 0>
     }]
    relativeSpaceBone {boneName=root}
    sphericalLimits [
     {limitRadius=7, drivingBone {boneName=upperarm_l}, sphereLocalOffset=<15 0 0>}
     {limitRadius=15, drivingBone {boneName=spine_04}, sphereLocalOffset=<0 5 0>}
    ]
   }
   n121 = Anim Dynamics - Bone: dyn_ear_l(ComponentPose=@n120, Alpha=1, ExternalForce="0,0,0") {
    angularDampingOverride=0.98
    angularSpringConstant=50
    bAngularSpring=true
    bGravityOverrideInSimSpace=true
    bOverrideAngularDamping=true
    bUsePlanarLimit=false
    simulationSpace=RootRelative
    alphaBoolBlend {blendOption=Linear}
    boundBone {boneName=dyn_ear_l}
    componentAppliedLinearAccClamp=<3500 3500 3500>
    componentLinearAccScale=<0.6 0.6 1.2>
    componentLinearVelScale=<0.6 0.6 1.2>
    physicsBodyDefinitions [{
      sphereCollisionRadius=0
      boundBone {boneName=dyn_ear_l}
      boxExtents=<20 15 15>
      constraintSetup {angularLimitsMax=<1 30 2>, angularLimitsMin=<-1 -30 -15>, angularTarget=<1 0 0>}
      localJointOffset=<1 0 0>
     }]
    relativeSpaceBone {boneName=root}
   }
   n122 = Anim Dynamics - Bone: dyn_ear_l_01(ComponentPose=@n121, Alpha=1, ExternalForce="0,0,0") {
    angularDampingOverride=0.98
    angularSpringConstant=50
    bAngularSpring=true
    bGravityOverrideInSimSpace=true
    bOverrideAngularDamping=true
    bUsePlanarLimit=false
    bUseSphericalLimits=true
    simulationSpace=RootRelative
    alphaBoolBlend {blendOption=Linear}
    boundBone {boneName=dyn_ear_l_01}
    componentAppliedLinearAccClamp=<3500 3500 3500>
    componentLinearAccScale=<0.6 0.6 1.2>
    componentLinearVelScale=<0.6 0.6 1.2>
    physicsBodyDefinitions [{
      collisionType=CustomSphere
      sphereCollisionRadius=2
      boundBone {boneName=dyn_ear_l_01}
      boxExtents=<20 15 15>
      constraintSetup {angularLimitsMax=<5 30 2>, angularLimitsMin=<-5 -30 -15>, angularTarget=<1 0 0>}
      localJointOffset=<1 0 0>
     }]
    relativeSpaceBone {boneName=root}
    sphericalLimits [{limitRadius=15.5, drivingBone {boneName=clavicle_l}, sphereLocalOffset=<2 -3 -6>}]
   }
   n123 = Anim Dynamics - Bone: dyn_ear_l_02(ComponentPose=@n122, Alpha=1, ExternalForce="0,0,0") {
    angularDampingOverride=0.98
    angularSpringConstant=50
    bAngularSpring=true
    bGravityOverrideInSimSpace=true
    bOverrideAngularDamping=true
    bUsePlanarLimit=false
    bUseSphericalLimits=true
    simulationSpace=RootRelative
    alphaBoolBlend {blendOption=Linear}
    boundBone {boneName=dyn_ear_l_02}
    componentAppliedLinearAccClamp=<3500 3500 3500>
    componentLinearAccScale=<0.6 0.6 1.2>
    componentLinearVelScale=<0.6 0.6 1.2>
    physicsBodyDefinitions [{
      collisionType=CustomSphere
      sphereCollisionRadius=1
      boundBone {boneName=dyn_ear_l_02}
      boxExtents=<20 15 15>
      constraintSetup {angularLimitsMax=<5 30 2>, angularLimitsMin=<-5 -30 -15>, angularTarget=<1 0 0>}
      localJointOffset=<1 0 0>
     }]
    relativeSpaceBone {boneName=root}
    sphericalLimits [{limitRadius=15.7, drivingBone {boneName=clavicle_l}, sphereLocalOffset=<4 -3 -9.5>}]
   }
   n124 = Anim Dynamics - Bone: dyn_ear_r(ComponentPose=@n123, Alpha=1, ExternalForce="0,0,0") {
    angularDampingOverride=0.98
    angularSpringConstant=50
    bAngularSpring=true
    bGravityOverrideInSimSpace=true
    bOverrideAngularDamping=true
    bUsePlanarLimit=false
    simulationSpace=RootRelative
    alphaBoolBlend {blendOption=Linear}
    boundBone {boneName=dyn_ear_r}
    componentAppliedLinearAccClamp=<3500 3500 3500>
    componentLinearAccScale=<0.6 0.6 1.2>
    componentLinearVelScale=<0.6 0.6 1.2>
    physicsBodyDefinitions [{
      sphereCollisionRadius=0
      boundBone {boneName=dyn_ear_r}
      boxExtents=<20 15 15>
      constraintSetup {angularLimitsMax=<1 30 2>, angularLimitsMin=<-1 -30 -15>, angularTarget=<1 0 0>}
      localJointOffset=<1 0 0>
     }]
    relativeSpaceBone {boneName=root}
   }
   n125 = Anim Dynamics - Bone: dyn_ear_r_01(ComponentPose=@n124, Alpha=1, ExternalForce="0,0,0") {
    angularDampingOverride=0.98
    angularSpringConstant=50
    bAngularSpring=true
    bGravityOverrideInSimSpace=true
    bOverrideAngularDamping=true
    bUsePlanarLimit=false
    bUseSphericalLimits=true
    simulationSpace=RootRelative
    alphaBoolBlend {blendOption=Linear}
    boundBone {boneName=dyn_ear_r_01}
    componentAppliedLinearAccClamp=<3500 3500 3500>
    componentLinearAccScale=<0.6 0.6 1.2>
    componentLinearVelScale=<0.6 0.6 1.2>
    physicsBodyDefinitions [{
      collisionType=CustomSphere
      sphereCollisionRadius=1
      boundBone {boneName=dyn_ear_r_01}
      boxExtents=<20 15 15>
      constraintSetup {angularLimitsMax=<5 30 2>, angularLimitsMin=<-5 -30 -15>, angularTarget=<1 0 0>}
      localJointOffset=<1 0 0>
     }]
    relativeSpaceBone {boneName=root}
    sphericalLimits [{limitRadius=15.5, drivingBone {boneName=clavicle_r}, sphereLocalOffset=<-2 3 6>}]
   }
   n126 = Anim Dynamics - Bone: dyn_ear_r_02(ComponentPose=@n125, Alpha=1, ExternalForce="0,0,0") {
    angularDampingOverride=0.98
    angularSpringConstant=50
    bAngularSpring=true
    bGravityOverrideInSimSpace=true
    bOverrideAngularDamping=true
    bUsePlanarLimit=false
    bUseSphericalLimits=true
    simulationSpace=RootRelative
    alphaBoolBlend {blendOption=Linear}
    boundBone {boneName=dyn_ear_r_02}
    componentAppliedLinearAccClamp=<3500 3500 3500>
    componentLinearAccScale=<0.6 0.6 1.2>
    componentLinearVelScale=<0.6 0.6 1.2>
    physicsBodyDefinitions [{
      collisionType=CustomSphere
      sphereCollisionRadius=1
      boundBone {boneName=dyn_ear_r_02}
      boxExtents=<20 15 15>
      constraintSetup {angularLimitsMax=<5 30 2>, angularLimitsMin=<-5 -30 -15>, angularTarget=<1 0 0>}
      localJointOffset=<1 0 0>
     }]
    relativeSpaceBone {boneName=root}
    sphericalLimits [{limitRadius=15.7, drivingBone {boneName=clavicle_r}, sphereLocalOffset=<-4 3 9.5>}]
   }
   Output Pose(Result=Component To Local(ComponentPose=@n126)) {layerGroup=None}
   n127 = RigidBody!(
    ComponentPose=Local To Component(LocalPose=Use cached pose 'pre rigid'())
    bAlphaBoolEnabled:bool=NOT Boolean(A=Get cape() -> (cape:bool))
    OverrideWorldGravity=Get GravityOverride()
    ExternalForce
    ComponentLinearAccScale:Vector=Get Acc() -> (Acc:Vector)
    ComponentLinearVelScale:Vector=Get Vel() -> (Vel:Vector)
   ) -> (Pose) {
    alphaInputType=Bool
    bClampLinearTranslationLimitToRefPose=true
    bOverrideWorldGravity=true
    overridePhysicsAsset=&/F_MED_PencilFig_Backbling_Hair_Physics
    simulationSpace=BaseBoneSpace
    alphaBoolBlend {blendInTime=0.1, blendOption=Linear, blendOutTime=0.1}
    baseBoneRef {boneName=spine_05}
    componentAppliedLinearAccClamp=<1500 1500 1500>
    componentLinearAccScale=<0.4 0.4 0.5>
    componentLinearVelScale=<0.4 0.4 0.5>
    overrideWorldGravity {z=-800}
   }
   n128 = RigidBody(
    ComponentPose=@n127
    bAlphaBoolEnabled=Get cape()
    OverrideWorldGravity=Get GravityOverride()
    ComponentLinearAccScale=Get Acc()
    ComponentLinearVelScale=Get Vel()
   ) {
    alphaInputType=Bool
    bClampLinearTranslationLimitToRefPose=true
    bOverrideWorldGravity=true
    overridePhysicsAsset=&/F_MED_PencilFig_Hair_Physics
    simulationSpace=BaseBoneSpace
    alphaBoolBlend {blendInTime=0.1, blendOption=Linear, blendOutTime=0.1}
    baseBoneRef {boneName=spine_05}
    componentAppliedLinearAccClamp=<1500 1500 1500>
    componentLinearAccScale=<0.4 0.4 0.5>
    componentLinearVelScale=<0.4 0.4 0.5>
    overrideWorldGravity {z=-800}
   }
   n129 = Component To Local(ComponentPose=@n128)
   n130 = RigidBody(
    ComponentPose=Local To Component(LocalPose=Use cached pose 'pre rigid'())
    OverrideWorldGravity=Get GravityOverride()
    ComponentLinearAccScale="0.4,0.4,0.5"
    ComponentLinearVelScale="0.4,0.4,0.5"
   ) {
    alphaInputType=Bool
    bClampLinearTranslationLimitToRefPose=true
    bOverrideWorldGravity=true
    overridePhysicsAsset=&/F_MED_PencilFig_skydive_Hair_Physics
    simulationSpace=BaseBoneSpace
    alphaBoolBlend {blendInTime=0.1, blendOption=Linear, blendOutTime=0.1}
    baseBoneRef {boneName=root}
    componentAppliedLinearAccClamp=<1500 1500 1500>
    componentLinearAccScale=<0.4 0.4 0.5>
    componentLinearVelScale=<0.4 0.4 0.5>
    overrideWorldGravity {z=-800}
   }
   n131 = Component To Local(ComponentPose=@n130)
   n132 = Get skydive no pack() -> (skydive no pack:bool)
   n133 = Get cape_constraints() -> (cape_constraints:double)
   n134 = "Constraint - dyn_hair_l_main_03 (PP)"(
    ComponentPose=Local To Component(LocalPose=Blend Poses by bool(BlendPose_0=@n131, BlendPose_1=@n129, bActiveValue=@n132) {blendType=Linear})
    Alpha=@n133
    ConstraintWeights_0=0.3
    ConstraintWeights_1=0.7
   ) {
    alphaBoolBlend {blendOption=Linear}
    boneToModify {boneName=dyn_hair_l_main_03}
    constraintSetup [{targetBone {boneName=spine_05}}, {targetBone {boneName=neck_01}}]
    constraintWeights [1,1]
   }
   n135 = "Constraint - dyn_hair_l_main_04 (PP)"(ComponentPose=@n134, Alpha=@n133, ConstraintWeights_0=0.5, ConstraintWeights_1=0.5) {
    alphaBoolBlend {blendOption=Linear}
    boneToModify {boneName=dyn_hair_l_main_04}
    constraintSetup [{targetBone {boneName=spine_05}}, {targetBone {boneName=neck_01}}]
    constraintWeights [1,1]
   }
   n136 = "Constraint - dyn_hair_l_main_05 (PP)"(ComponentPose=@n135, Alpha=@n133, ConstraintWeights_0=0.7, ConstraintWeights_1=0.3) {
    alphaBoolBlend {blendOption=Linear}
    boneToModify {boneName=dyn_hair_l_main_05}
    constraintSetup [{targetBone {boneName=spine_05}}, {targetBone {boneName=spine_04}}]
    constraintWeights [1,1]
   }
   n137 = "Constraint - dyn_hair_l_main_06 (PP)"(ComponentPose=@n136, Alpha=@n133, ConstraintWeights_0=0.5, ConstraintWeights_1=0.5) {
    alphaBoolBlend {blendOption=Linear}
    boneToModify {boneName=dyn_hair_l_main_06}
    constraintSetup [{targetBone {boneName=spine_05}}, {targetBone {boneName=spine_04}}]
    constraintWeights [1,1]
   }
   n138 = "Constraint - dyn_hair_r_main_03 (PP)"(ComponentPose=@n137, Alpha=@n133, ConstraintWeights_0=0.3, ConstraintWeights_1=0.7) {
    alphaBoolBlend {blendOption=Linear}
    boneToModify {boneName=dyn_hair_r_main_03}
    constraintSetup [{targetBone {boneName=spine_05}}, {targetBone {boneName=neck_01}}]
    constraintWeights [1,1]
   }
   n139 = "Constraint - dyn_hair_r_main_04 (PP)"(ComponentPose=@n138, Alpha=@n133, ConstraintWeights_0=0.5, ConstraintWeights_1=0.5) {
    alphaBoolBlend {blendOption=Linear}
    boneToModify {boneName=dyn_hair_r_main_04}
    constraintSetup [{targetBone {boneName=spine_05}}, {targetBone {boneName=neck_01}}]
    constraintWeights [1,1]
   }
   n140 = "Constraint - dyn_hair_r_main_05 (PP)"(ComponentPose=@n139, Alpha=@n133, ConstraintWeights_0=0.7, ConstraintWeights_1=0.3) {
    alphaBoolBlend {blendOption=Linear}
    boneToModify {boneName=dyn_hair_r_main_05}
    constraintSetup [{targetBone {boneName=spine_05}}, {targetBone {boneName=spine_04}}]
    constraintWeights [1,1]
   }
   n141 = "Constraint - dyn_hair_r_main_06 (PP)"(ComponentPose=@n140, Alpha=@n133, ConstraintWeights_0=0.5, ConstraintWeights_1=0.5) {
    alphaBoolBlend {blendOption=Linear}
    boneToModify {boneName=dyn_hair_r_main_06}
    constraintSetup [{targetBone {boneName=spine_05}}, {targetBone {boneName=spine_04}}]
    constraintWeights [1,1]
   }
   Save cached pose 'Perfmode'(Pose=Component To Local(ComponentPose=@n141)) {cacheName=Perfmode}
   n142 = Control Rig(Source=Input Pose()) {
    alphaBoolBlend {blendOption=Linear}
    controlRigAssetReference {blueprintRigClass=&/F_MED_PencilFig_Hair_CtrlRig.F_MED_PencilFig_Hair_CtrlRig_C}
   }
   n143 = Look At - Bone: dyn_hair_r_main_Look_at(ComponentPose=Local To Component(LocalPose=@n142), Alpha=0.7) {
    alphaBoolBlend {blendOption=Linear}
    boneToModify {boneName=dyn_hair_r_main_Look_at}
    lookAtLocation {x=0}
    lookAtTarget {boneReference {boneName=dyn_hair_r_TGT}}
    lookAt_Axis {axis=<1 0>}
   }
   n144 = Look At - Bone: dyn_hair_l_main_Look_at(ComponentPose=@n143, Alpha=0.7) {
    alphaBoolBlend {blendOption=Linear}
    boneToModify {boneName=dyn_hair_l_main_Look_at}
    lookAtLocation {x=0}
    lookAtTarget {boneReference {boneName=dyn_hair_l_TGT}}
    lookAt_Axis {axis=<1 0>}
   }
   n145 = Get jumping() -> (jumping:double)
   n146 = "Transform (Modify) Bone - Bone: dyn_hair_r_main_Look_at"(ComponentPose=@n144, Alpha=@n145, Translation="0,0,0", Rotation="0,-35,0", Scale="1,1,1") {
    rotationMode=BMM_Additive
    rotationSpace=BCS_BoneSpace
    alphaBoolBlend {blendOption=Linear}
    alphaScaleBiasClamp {bInterpResult=true, interpSpeedIncreasing=2}
    boneToModify {boneName=dyn_hair_r_main_Look_at}
    rotation {yaw=-35}
   }
   n147 = "Transform (Modify) Bone - Bone: dyn_hair_l_main_Look_at"(ComponentPose=@n146, Alpha=@n145, Translation="0,0,0", Rotation="0,-35,0", Scale="1,1,1") {
    rotationMode=BMM_Additive
    rotationSpace=BCS_BoneSpace
    alphaBoolBlend {blendOption=Linear}
    alphaScaleBiasClamp {bInterpResult=true, interpSpeedIncreasing=2}
    boneToModify {boneName=dyn_hair_l_main_Look_at}
    rotation {yaw=-35}
   }
   Save cached pose 'pre rigid'(Pose=Component To Local(ComponentPose=@n147)) {cacheName=pre rigid}
   NOT Boolean(A=Get cape())
   Get Vel()
   Get Acc()
  ]}
 EventGraph {nodes [
   Event BlueprintUpdateAnimation()
   > n1 = Sequence()
   > n2 = Gravity Override State Function DEPRECATED(
    CharacterPartAnimInstance=Self-Reference()
    GameTimeInSeconds=GetGameTimeInSeconds()
    GravityOverrideParams=Get GravityOverrideParams() -> (GravityOverrideParams:Gravity Override Params Struct Structure)
   )
   > Set GravityOverride(GravityOverride=@n2)
   > n3 = Branch(Condition=CheckForBackbling())
   > Set Vel(Vel:Vector="0, 0, 0") -> (Output_Get:Vector)
   > Set Acc(Acc:Vector="0, 0, 0") -> (Output_Get:Vector)
   > Set Hair Anim Dyn A(Hair Anim Dyn A:double=0.1) -> (Output_Get:double)
   > Set Hair Anim Dyn B(Hair Anim Dyn B:double=0.1) -> (Output_Get:double)
   > Set Hair Anim Dyn C(Hair Anim Dyn C:double=0.1) -> (Output_Get:double)
   > n4 = Branch(Condition=CheckForBackbling().bIsBackblingACape)
   > Set cape(cape:bool=true) -> (Output_Get:bool)
   > Set cape_constraints(cape_constraints:double=0.5) -> (Output_Get:double)
   Set cape(execute=@n4.else, cape=false)
   > Set cape_constraints(cape_constraints=0)
   Set Vel(execute=@n3.else, Vel="0.4,0.4,0.5")
   > Set Acc(Acc="0.4,0.4,0.5")
   > Set Hair Anim Dyn A(Hair Anim Dyn A=0.5)
   > Set Hair Anim Dyn B(Hair Anim Dyn B=0.4)
   > Set Hair Anim Dyn C(Hair Anim Dyn C=0.2)
   > Set cape(cape=true)
   > Set cape_constraints(cape_constraints=0)
   n5 = Branch(execute=@n1.then_1, Condition=AND Boolean(A=Get bIsSkyDiving(), B=Get cape()))
   > Set skydive no pack(skydive no pack:bool=true) -> (Output_Get:bool)
   Set skydive no pack(execute=@n5.else, skydive no pack=false)
   n6 = Event OnBeginFalling()
   Set jumping(execute=[@n6.then, @n8.then], jumping:double=1) -> (Output_Get:double)
   n7 = Event OnEndFalling()
   Set jumping(execute=[@n7.then, @n9.then], jumping=0)
   n8 = Event OnBeginJumping()
   n9 = Event OnEndJumping()
   Branch()
  ]}
]
variables [
 Acc {type=Vector}
 cape {type=boolean}
 cape_constraints {type=number}
 GravityOverride {type=Vector}
 GravityOverrideParams {type=GravityOverrideParamsStruct, default {skydive_FloatingDown {gravityOverride=<0 -150 600>, windFrequency=<0.75 3.5 4>}}}
 Hair Anim Dyn A {type=number}
 Hair Anim Dyn B {type=number}
 Hair Anim Dyn C {type=number}
 jumping {type=number}
 skydive no pack {type=boolean}
 Vel {type=Vector}
]
